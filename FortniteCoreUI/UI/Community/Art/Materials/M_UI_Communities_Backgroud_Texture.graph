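# Translucent UI background material for Communities that builds a layered look from a single ImageTexture parameter.
# MF_UI_UVScaling supplies a ScaleY-mapped base, whose color is squared, and scaled-to-fit UVs that CornerTextureScale rescales about the top-right pivot.
# That corner sample is lerped over the base using its alpha, passed through an edge-pixel clip to limit it to its bounds.
# A CoreUI rounded mask function supplies a rounded-corner shape.
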Material /FortniteCoreUI/UI/Community/Art/Materials/M_UI_Communities_Backgroud_Texture
graph {
 properties {
  bAutomaticallySetUsageInEditor=false
  bUsedWithStaticMesh=false
  blendMode=BLEND_Translucent
  materialDomain=MD_UI
  usedShadingModels=Default Lit
 }
 nodes [
  n1 = TextureObjectParameter() {parameterName=ImageTexture, texture=/Game/Global/Textures/Default/Blanks/T_Blank_White_ZeroAlpha_sRGBColor}
  n2 = MaterialFunctionCall(
   UV=MaterialFunctionCall() {materialFunction=&GetUserInterfaceUV}.Normalized UV
   TextureSize=TextureProperty(Tex=@n1)
   ScalePivotPoint=Constant2Vector() {r=1}
  ) {materialFunction=&MF_UI_UVScaling}
  n3 = TextureSample(UVs=@n2.ScaleY, Tex=@n1)
  n4 = Constant2Vector() {r=1}
  n5 = Add(A=Multiply(A=Subtract(A=@n2.Scaled to Fit, B=@n4), B=ScalarParameter() {defaultValue=2, parameterName=CornerTextureScale}) {constB=2}, B=@n4)
  n6 = TextureSample(UVs=@n5, Tex=@n1)
  LinearInterpolate(
   A=Multiply(A=@n3.RGB, B=@n3.RGB)
   B=@n6.RGB
   Alpha=MaterialFunctionCall!(UVs=@n5, Alpha=@n6.A) {materialFunction=/FortUILibrary/MaterialFunctions/Basics/MF_UI_ClipEdgePixel}.Result
  )
  MaterialFunctionCall(Corner Radius=Constant() {r=10}, Radius in Pixels=Constant() {r=1}) {materialFunction=&MF_UI_CoreUI_RoundedMask}
 ]
}
