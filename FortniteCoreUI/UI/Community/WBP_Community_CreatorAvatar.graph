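# Community creator avatar widget that shows an image loaded from a URL in a size box around a lazy image.
# PreConstruct applies the configurable ImageSize to the box overrides and the rebuilt brush, and sets the material texture parameter name.
# SetAvatarImageURL starts the download and clears the empty state.
# Load errors set the avatar material's "Is Empty" scalar parameter, so the material handles the fallback look.
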WidgetBlueprint /FortniteCoreUI/UI/Community/WBP_Community_CreatorAvatar
graphs [
 ApplyImageSize {nodes [
   n1 = ApplyImageSize!() -> (then, Size:Vector 2D Structure) {is=title K2Node_FunctionEntry}
   n2 = BreakVector2D(InVec=@n1)
   n3 = Get SB_Image() -> (SB_Image:Size Box Object Reference)
   > SetWidthOverride(self=@n3, InWidthOverride=@n2)
   > SetHeightOverride(self=@n3, InHeightOverride=@n2.Y)
   n4 = Get LazyImage_Avatar() -> (LazyImage_Avatar:Fort Lazy Image Object Reference)
   n5 = Break Slate Brush!(SlateBrush:"Slate Brush Structure (by ref)"=Get Brush!(self:Image Object Reference=@n4) -> (Brush:Slate Brush Structure)) -> (
    TintColor:Slate Color Structure
    DrawAs:ESlateBrushDrawType Enum
    Tiling:ESlateBrushTileType Enum
    Mirroring:ESlateBrushMirrorType Enum
    ImageSize:Vector 2D Structure
    Margin:Margin Structure
    ResourceObject:Object Reference
    OutlineSettings:Slate Brush Outline Settings Structure
   ) {is=title K2Node_BreakStruct}
   > Set Brush!(
    execute
    Brush:Slate Brush Structure=Make Slate Brush!(
     TintColor:Slate Color Structure=@n5
     DrawAs:ESlateBrushDrawType Enum=@n5.DrawAs
     Tiling:ESlateBrushTileType Enum=@n5.Tiling
     Mirroring:ESlateBrushMirrorType Enum=@n5.Mirroring
     ImageSize:Vector 2D Structure=@n1
     Margin:Margin Structure=@n5.Margin
     ResourceObject:Object Reference=@n5.ResourceObject
     OutlineSettings:Slate Brush Outline Settings Structure=@n5.OutlineSettings
    ) -> (SlateBrush:Slate Brush Structure) {is=title K2Node_MakeStruct}
    self:Image Object Reference=@n4
   ) -> (then, Output_Get:Slate Brush Structure)
   > SetMaterialTextureParamName!(execute, self:Common Lazy Image Object Reference=@n4, TextureParamName:Name=Get MaterialParamName() -> (MaterialParamName:Name=None)) -> (then)
  ]}
 DisplayEmptyImage {nodes [
   n1 = DisplayEmptyImage!() -> (then, IsEmpty:bool=~) {is=title K2Node_FunctionEntry}
   > n2 = GetDynamicMaterial(self=Get LazyImage_Avatar())
   > SetScalarParameterValue(self=@n2, ParameterName=Is Empty, Value="To Float (Boolean)"(InBool=@n1))
  ]}
 EventGraph {nodes [
   "On Loading Errored (LazyImage_Avatar)"!() -> (OutputDelegate, then, ErrorCode:int=0, ModuleId:int=0, ErrorString:String, ErrorMessage:Text) {is=title K2Node_ComponentBoundEvent}
   > DisplayEmptyImage!(execute, self, IsEmpty:bool=true) -> (then) {is=title K2Node_CallFunction}
   Event PreConstruct()
   > ApplyImageSize!(execute, self, Size:Vector 2D Structure=Get ImageSize() -> (ImageSize:Vector 2D Structure)) -> (then) {is=title K2Node_CallFunction}
  ]}
 SetAvatarImageURL {nodes [
   n1 = SetAvatarImageURL!() -> (then, URL:String) {is=title K2Node_FunctionEntry}
   > SetBrushFromURL!(
    execute
    self:Fort Lazy Image Object Reference=Get LazyImage_Avatar()
    CMSImageURL:String=@n1
    bMatchSize:bool=false
    Delay:float=0
    DownloadPriority:ECMSDownloadPriority Enum=Normal
   ) -> (then)
   > DisplayEmptyImage(IsEmpty=false)
  ]}
]
variables [
 ImageSize {type=Vector2D, default=<32 32>}
 MaterialParamName {type=string, default=Thumbnail}
]
widgets {
 info {inheritedWidgetCount=0, parentClass=&UserWidget, rootWidgetClass=&SizeBox, widgetCount=2}
 widgets [
  SB_Image {widgetClassPath=&SizeBox, bIsVariable=true, widget=SB_Image}
  LazyImage_Avatar {
   widgetClassPath=&FortLazyImage
   bIsVariable=true
   parent=SB_Image
   widget=LazyImage_Avatar
   properties {
    materialTextureParamName=Thumbnail
    brush {
     bIsDynamicallyLoaded=false
     drawAs=Image
     imageType=FullColor
     mirroring=NoMirror
     resourceName=None
     resourceObject=/FortniteCoreUI/UI/Community/Art/MaterialInstances/MI_UI_Community_Avatar_Presentation
     tiling=NoTile
     imageSize=<32 32>
     margin=<ltrb 0 0 0 0>
     outlineSettings {
      bUseBrushTransparency=false
      roundingType=HalfHeightRadius
      width=0
      color {colorUseRule=UseColor_Specified, specifiedColor=<rgba 0 0 0 0>}
      cornerRadii=<0 0 0 0>
     }
     tintColor {colorUseRule=UseColor_Specified, specifiedColor=<rgba 1 1 1 1>}
     uVRegion {bIsValid=false, max=<0 0>, min=<0 0>}
    }
    loadingBackgroundBrush {
     bIsDynamicallyLoaded=false
     drawAs=Image
     imageType=FullColor
     mirroring=NoMirror
     resourceName=None
     resourceObject=/FortniteCoreUI/UI/Community/Art/MaterialInstances/MI_UI_Community_Avatar_Loading
     tiling=NoTile
     imageSize=<32 32>
     margin=<ltrb 0 0 0 0>
     outlineSettings {
      bUseBrushTransparency=false
      roundingType=HalfHeightRadius
      width=0
      color {colorUseRule=UseColor_Specified, specifiedColor=<rgba 0 0 0 0>}
      cornerRadii=<0 0 0 0>
     }
     tintColor {colorUseRule=UseColor_Specified, specifiedColor=<rgba 1 1 1 1>}
     uVRegion {bIsValid=false, max=<0 0>, min=<0 0>}
    }
    loadingThrobberBrush {
     bIsDynamicallyLoaded=false
     drawAs=Image
     imageType=FullColor
     mirroring=NoMirror
     resourceName=None
     resourceObject=/Game/Global/Textures/Default/Blanks/T_Blank_White_ZeroAlpha_sRGBColor
     tiling=NoTile
     imageSize=<1 1>
     margin=<ltrb 0 0 0 0>
     outlineSettings {
      bUseBrushTransparency=false
      roundingType=HalfHeightRadius
      width=0
      color {colorUseRule=UseColor_Specified, specifiedColor=<rgba 0 0 0 0>}
      cornerRadii=<0 0 0 0>
     }
     tintColor {colorUseRule=UseColor_Specified, specifiedColor=<rgba 1 1 1 1>}
     uVRegion {bIsValid=false, max=<0 0>, min=<0 0>}
    }
   }
   slot {class=SizeBoxSlot}
  }
 ]
}
