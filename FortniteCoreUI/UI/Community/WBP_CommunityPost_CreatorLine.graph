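# Community post byline widget that shows a creator avatar, creator name, island name and relative time on one wrapping line.
# A WrapBox holds the WBP_Community_CreatorAvatar and CommonTextBlocks separated by dot dividers, with muted grey on all text but the creator name.
# The event graph has no logic; avatar, name, island and time are exposed as variables.
# Compact and default separator padding variables are declared but unused here.
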WidgetBlueprint /FortniteCoreUI/UI/Community/WBP_CommunityPost_CreatorLine
graphs [EventGraph {nodes [
   Event PreConstruct()
   Event Construct()
   Event Tick()
  ]}]
variables [
 TextSeparator_Padding_CompactUI {type=Margin, default=<ltrb 8 0 8 0>}
 TextSeparator_Padding_Default {type=Margin, default=<ltrb 4 0 4 0>}
]
widgets {
 info {inheritedWidgetCount=0, parentClass=&UserWidget, rootWidgetClass=&WrapBox, widgetCount=7}
 widgets [
  WrapBox_2 {widgetClassPath=&WrapBox, widget=WrapBox_2, properties {bOverride_Cursor=true}}
  CreatorAvatar {
   widgetClassPath=&/WBP_Community_CreatorAvatar
   bIsVariable=true
   parent=WrapBox_2
   widget=CreatorAvatar
   slot {class=WrapBoxSlot, properties {padding=<ltrb 0 0 8 0>}}
  }
  Text_Creator {
   widgetClassPath=&CommonTextBlock
   bIsVariable=true
   parent=WrapBox_2
   widget=Text_Creator
   properties {
    mobileFontSizeMultiplier=1.4
    text=Creator
    font {
     bForceMonospaced=false
     bMaterialIsStencil=false
     fontMaterial=null
     fontObject=&HeadingNow
     letterSpacing=10
     monospacedWidth=1
     size=14.25
     skewAmount=0
     typefaceFontName=74 Regular
     outlineSettings {
      bApplyOutlineToDropShadows=false
      bMiteredCorners=false
      bSeparateFillAlpha=false
      outlineMaterial=null
      outlineSize=0
      outlineColor=<rgba 0 0 0 1>
     }
    }
    shadowOffset=<0 0>
   }
   slot {class=WrapBoxSlot, properties {verticalAlignment=VAlign_Center}}
  }
  Text_Divider {
   widgetClassPath=&CommonTextBlock
   parent=WrapBox_2
   widget=Text_Divider
   properties {
    mobileFontSizeMultiplier=1.4
    text="·"
    colorAndOpacity {colorUseRule=UseColor_Specified, specifiedColor=<rgba 0.5776 0.5776 0.5776 1>}
    font {
     bForceMonospaced=false
     bMaterialIsStencil=false
     fontMaterial=null
     fontObject=&HeadingNow
     letterSpacing=10
     monospacedWidth=1
     size=11.25
     skewAmount=0
     typefaceFontName=74 Regular
     outlineSettings {
      bApplyOutlineToDropShadows=false
      bMiteredCorners=false
      bSeparateFillAlpha=false
      outlineMaterial=null
      outlineSize=0
      outlineColor=<rgba 0 0 0 1>
     }
    }
    shadowOffset=<0 0>
   }
   slot {class=WrapBoxSlot, properties {verticalAlignment=VAlign_Center, padding=<ltrb 4 0 4 0>}}
  }
  Text_IslandName {
   widgetClassPath=&CommonTextBlock
   bIsVariable=true
   parent=WrapBox_2
   widget=Text_IslandName
   properties {
    mobileFontSizeMultiplier=1.4
    text=Island Name
    colorAndOpacity {colorUseRule=UseColor_Specified, specifiedColor=<rgba 0.5776 0.5776 0.5776 1>}
    font {
     bForceMonospaced=false
     bMaterialIsStencil=false
     fontMaterial=null
     fontObject=&HeadingNow
     letterSpacing=10
     monospacedWidth=1
     size=14.25
     skewAmount=0
     typefaceFontName=74 Regular
     outlineSettings {
      bApplyOutlineToDropShadows=false
      bMiteredCorners=false
      bSeparateFillAlpha=false
      outlineMaterial=null
      outlineSize=0
      outlineColor=<rgba 0 0 0 1>
     }
    }
    shadowOffset=<0 0>
   }
   slot {class=WrapBoxSlot, properties {verticalAlignment=VAlign_Center}}
  }
  Text_Divider_1 {
   widgetClassPath=&CommonTextBlock
   parent=WrapBox_2
   widget=Text_Divider_1
   properties {
    mobileFontSizeMultiplier=1.4
    text="·"
    colorAndOpacity {colorUseRule=UseColor_Specified, specifiedColor=<rgba 0.5776 0.5776 0.5776 1>}
    font {
     bForceMonospaced=false
     bMaterialIsStencil=false
     fontMaterial=null
     fontObject=&HeadingNow
     letterSpacing=10
     monospacedWidth=1
     size=11.25
     skewAmount=0
     typefaceFontName=74 Regular
     outlineSettings {
      bApplyOutlineToDropShadows=false
      bMiteredCorners=false
      bSeparateFillAlpha=false
      outlineMaterial=null
      outlineSize=0
      outlineColor=<rgba 0 0 0 1>
     }
    }
    shadowOffset=<0 0>
   }
   slot {class=WrapBoxSlot, properties {verticalAlignment=VAlign_Center, padding=<ltrb 4 0 4 0>}}
  }
  Text_Time {
   widgetClassPath=&CommonTextBlock
   bIsVariable=true
   parent=WrapBox_2
   widget=Text_Time
   properties {
    mobileFontSizeMultiplier=1.4
    text=time ago
    colorAndOpacity {colorUseRule=UseColor_Specified, specifiedColor=<rgba 0.5776 0.5776 0.5776 1>}
    font {
     bForceMonospaced=false
     bMaterialIsStencil=false
     fontMaterial=null
     fontObject=&HeadingNow
     letterSpacing=10
     monospacedWidth=1
     size=14.25
     skewAmount=0
     typefaceFontName=74 Regular
     outlineSettings {
      bApplyOutlineToDropShadows=false
      bMiteredCorners=false
      bSeparateFillAlpha=false
      outlineMaterial=null
      outlineSize=0
      outlineColor=<rgba 0 0 0 1>
     }
    }
    shadowOffset=<0 0>
   }
   slot {class=WrapBoxSlot, properties {verticalAlignment=VAlign_Center}}
  }
 ]
}
