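# Community event date badge widget that shows a short month name and numeric day over a material-driven background.
# PreConstruct applies the Size variable to the root SizeBox, and Construct caches the background's dynamic material.
# SetDateText fills both labels, while SetDateSideColor writes the material's Side Color parameter.
# GetDateSideColor returns grey for ended or canceled events, otherwise one of two colors based on interest.
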WidgetBlueprint /FortniteCoreUI/UI/Community/WBP_Community_Date
graphs [
 EventGraph {nodes [
   Event PreConstruct()
   > SetSize!(execute, self, Size:Vector 2D Structure=Get Size() -> (Size:Vector 2D Structure)) -> (then)
   Event Construct()
   > n1 = GetDynamicMaterial(self=Get Background() -> (Background:Image Object Reference))
   > Set MID(MID:Material Instance Dynamic Object Reference=@n1) -> (Output_Get:Material Instance Dynamic Object Reference)
  ]}
 GetDateSideColor {nodes [
   n1 = GetDateSideColor!() -> (then, IsInterested:bool=~, EventStatus:ECommunityEventStatus Enum=Unknown) {is=title K2Node_FunctionEntry}
   n2 = Get EventStatus() -> (EventStatus:ECommunityEventStatus Enum=Unknown)
   n3 = "Equal (Enum)"(A:ECommunityEventStatus Enum=@n2, B:ECommunityEventStatus Enum=Ended)
   n4 = "Equal (Enum)"(A=@n2, B=Canceled)
   > Return Node!(
    execute
    ReturnValue:Linear Color Structure=Select(
     Option 0:Linear Color Structure=Select(Option 0=(R=0,G=1,B=0.6038,A=1), Option 1=(R=0.9301,G=1,B=0.0097,A=1), Index=@n1)
     Option 1:Linear Color Structure=(R=0.5176,G=0.5176,B=0.5176,A=1)
     Index=OR Boolean(A=@n4, B=@n3)
    ) -> (ReturnValue:Linear Color Structure)
   )
  ]}
 SetDateSideColor {nodes [
   n1 = SetDateSideColor!() -> (then, Color:Linear Color Structure) {is=title K2Node_FunctionEntry}
   > SetVectorParameterValue(self=Get MID() -> (MID:Material Instance Dynamic Object Reference), ParameterName=Side Color, Value=@n1)
  ]}
 SetDateText {nodes [
   n1 = SetDateText!() -> (then, DateTime:Date Time Structure) {is=title K2Node_FunctionEntry}
   > "SetText (Text)"(
    self=Get Text_Month() -> (Text_Month:Common Text Object Reference)
    InText=AsMonth!(InDateTime:"Date Time Structure (by ref)"=@n1, Style:EMonthDisplayStyle Enum=Short) -> (ReturnValue:Text)
   )
   > "SetText (Text)"(
    self=Get Text_Day() -> (Text_Day:Common Text Object Reference)
    InText=AsDayOfMonth!(InDateTime:"Date Time Structure (by ref)"=Get DateTime() -> (DateTime:Date Time Structure), Style:EDayOfMonthDisplayStyle Enum=Numeric) -> (ReturnValue:Text)
   )
  ]}
 SetSize {nodes [
   n1 = SetSize!() -> (then, Size:Vector 2D Structure) {is=title K2Node_FunctionEntry}
   n2 = Get SB_Content() -> (SB_Content:Size Box Object Reference)
   n3 = BreakVector2D(InVec=@n1)
   > SetWidthOverride(self=@n2, InWidthOverride=@n3)
   > SetHeightOverride(self=@n2, InHeightOverride=@n3.Y)
  ]}
]
variables [
 MID {type=MaterialInstanceDynamic}
 Size {type=Vector2D, default=<72 116>}
]
widgets {
 info {inheritedWidgetCount=0, parentClass=&CommonUserWidget, rootWidgetClass=&SizeBox, widgetCount=6}
 widgets [
  SB_Content {
   widgetClassPath=&SizeBox
   bIsVariable=true
   widget=SB_Content
   properties {bOverride_HeightOverride=true, bOverride_WidthOverride=true}
  }
  Overlay_30 {widgetClassPath=&Overlay, parent=SB_Content, widget=Overlay_30, slot {class=SizeBoxSlot}}
  Background {
   widgetClassPath=&Image
   bIsVariable=true
   parent=Overlay_30
   widget=Background
   properties {
    visibility=SelfHitTestInvisible
    brush {
     bIsDynamicallyLoaded=false
     drawAs=Image
     imageType=FullColor
     mirroring=NoMirror
     resourceName=None
     resourceObject=/FortniteCoreUI/UI/Community/Art/MaterialInstances/MI_UI_Community_Date
     tiling=NoTile
     imageSize=<72 116>
     margin=<ltrb 0 0 0 0>
     outlineSettings {
      bUseBrushTransparency=false
      roundingType=HalfHeightRadius
      width=0
      color {colorUseRule=UseColor_Specified, specifiedColor=<rgba 0 0 0 0>}
      cornerRadii=<0 0 0 0>
     }
     tintColor {colorUseRule=UseColor_Specified, specifiedColor=<rgba 1 1 1 1>}
     uVRegion {bIsValid=false, max=<0 0>, min=<0 0>}
    }
   }
   slot {class=OverlaySlot, properties {horizontalAlignment=HAlign_Fill, verticalAlignment=VAlign_Fill}}
  }
  StackBox_0 {
   widgetClassPath=&StackBox
   parent=Overlay_30
   widget=StackBox_0
   properties {orientation=Orient_Vertical}
   slot {class=OverlaySlot, properties {horizontalAlignment=HAlign_Fill, verticalAlignment=VAlign_Center}}
  }
  Text_Month {
   widgetClassPath=&CommonTextBlock
   bIsVariable=true
   parent=StackBox_0
   widget=Text_Month
   properties {
    autoWrapText=true
    justification=Center
    mobileFontSizeMultiplier=1.4
    text=Dec
    wrappingPolicy=AllowPerCharacterWrapping
    font {
     bForceMonospaced=false
     bMaterialIsStencil=false
     fontMaterial=null
     fontObject=&HeadingNow
     letterSpacing=0
     monospacedWidth=1
     size=15.75
     skewAmount=0
     typefaceFontName=75 Medium
     outlineSettings {
      bApplyOutlineToDropShadows=false
      bMiteredCorners=false
      bSeparateFillAlpha=false
      outlineMaterial=null
      outlineSize=0
      outlineColor=<rgba 0 0 0 1>
     }
    }
    shadowOffset=<0 0>
   }
   slot {class=StackBoxSlot, properties {padding=<ltrb 5 0 5 0>}}
  }
  Text_Day {
   widgetClassPath=&CommonTextBlock
   bIsVariable=true
   parent=StackBox_0
   widget=Text_Day
   properties {
    autoWrapText=true
    clipping=ClipToBounds
    justification=Center
    mobileFontSizeMultiplier=1.2
    text=32
    font {
     bForceMonospaced=false
     bMaterialIsStencil=false
     fontMaterial=null
     fontObject=&HeadingNow
     letterSpacing=0
     monospacedWidth=1
     size=15.75
     skewAmount=0
     typefaceFontName=75 Medium
     outlineSettings {
      bApplyOutlineToDropShadows=false
      bMiteredCorners=false
      bSeparateFillAlpha=false
      outlineMaterial=null
      outlineSize=0
      outlineColor=<rgba 0 0 0 1>
     }
    }
    shadowOffset=<0 0>
   }
   slot {class=StackBoxSlot, properties {padding=<ltrb 5 0 5 0>}}
  }
 ]
}
