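# Community event post date-info widget that shows a date tile beside a status block for upcoming, live, ended or canceled events.
# A visibility switcher picks the block from ECommunityEventStatus, with Unknown falling back to Upcoming.
# Setter functions fill the date tile, live highlight and community name labels.
# ConvertTimespanToTimerText returns localized, pluralized text in the largest nonzero unit from months to minutes, approximating months as 30 days.
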WidgetBlueprint /FortniteCoreUI/UI/Community/WBP_CommunityPost_DateInfo
graphs [
 ConvertTimespanToTimerText {nodes [
   n1 = ConvertTimespanToTimerText!() -> (then, Timespan:Timespan Structure) {is=title K2Node_FunctionEntry}
   n2 = Get Total Days!(A:Timespan Structure=@n1) -> (ReturnValue:double) {is=title K2Node_CallFunction}
   n3 = "Absolute (Integer)"!(A:int=int / int(A=Floor(A=@n2) {is=title K2Node_CallFunction}, B=30)) -> (ReturnValue:int=0)
   n4 = "Absolute (Integer)"(A=Round(A=Get Total Days(A=@n1)) {is=title K2Node_CallFunction}) -> (ReturnValue=0)
   n5 = Format Text!(Format="LOCTEXT(\"{Days} {Days}|plural(one=Day,other=Days)\")", Days:int=@n4) -> (Result)
   n6 = "integer > integer"(A=@n4)
   n7 = Get Minutes!(A:Timespan Structure=@n1) -> (ReturnValue:int=0) {is=title K2Node_CallFunction}
   n8 = "Absolute (Integer)"(A=Floor(A=Get Total Hours!(A:Timespan Structure=@n1) -> (ReturnValue:double) {is=title K2Node_CallFunction})) -> (ReturnValue=0)
   n9 = SelectText(
    A=Format Text!(Format="LOCTEXT(\"{Time} {Time}|plural(one=Minute,other=Minutes)\")", Time:int="Absolute (Integer)"(A=@n7) -> (ReturnValue=0)) -> (Result)
    B=Format Text(Format="LOCTEXT(\"{Time} {Time}|plural(one=Hour,other=Hours)\")", Time=@n8)
    bPickA="integer <= integer"(A=@n8)
   )
   > Return Node!(
    execute
    Text:Text=SelectText(
     A=Format Text(Format="LOCTEXT(\"{Time} {Time}|plural(one=Month,other=Months)\")", Time=@n3)
     B=SelectText(A=@n5, B=@n9, bPickA=@n6)
     bPickA="integer > integer"(A=@n3)
    )
   )
  ]}
 EventGraph {nodes [
   n1 = Event PreConstruct()
   > Parent: PreConstruct(IsDesignTime=@n1.IsDesignTime)
   > SetFontSize!(
    execute
    self:Text Object Reference=Get Text_Highlight!(self:WBP UIKit Highlight Base Object Reference=Get Highlight() -> (Highlight:WBP UIKit Highlight Object Reference)) -> (Text_Highlight:Common Text Object Reference)
    DisplayFontSize:float=16
   ) -> (then)
   Event Construct()
   Event Tick()
  ]}
 SetCommunityName {nodes [
   n1 = SetCommunityName!() -> (then, CommunityName:Text) {is=title K2Node_FunctionEntry}
   > "SetText (Text)"(
    self=[
     Get Text_CommunityName() -> (Text_CommunityName:Common Text Object Reference)
     Get Text_CommunityName_2() -> (Text_CommunityName_2:Common Text Object Reference)
     Get Text_CommunityName_1() -> (Text_CommunityName_1:Common Text Object Reference)
    ]
    InText=Format Text!(Format="LOCTEXT(\"In {community name}\")", community name:Text=@n1) -> (Result)
   )
  ]}
 SetDateSwitcherState {nodes [
   SetDateSwitcherState!() -> (then, EventStatus:ECommunityEventStatus Enum=Unknown) {is=title K2Node_FunctionEntry}
   n1 = Get VB_Upcoming() -> (VB_Upcoming:Vertical Box Object Reference)
   > SetActiveWidget(
    self=Get StateSwitcher() -> (StateSwitcher:Common Visibility Switcher Object Reference)
    Widget=Select!(
     Unknown:Widget Object Reference=@n1
     Upcoming:Widget Object Reference=@n1
     Live:Widget Object Reference=Get VB_Live() -> (VB_Live:Vertical Box Object Reference)
     Ended:Widget Object Reference=Get VB_Ended() -> (VB_Ended:Vertical Box Object Reference)
     Canceled:Widget Object Reference=Get VB_Canceled() -> (VB_Canceled:Vertical Box Object Reference)
     Index:ECommunityEventStatus Enum=Get EventStatus() -> (EventStatus:ECommunityEventStatus Enum=Unknown)
    ) -> (ReturnValue:Widget Object Reference)
   )
  ]}
 SetDateText {nodes [
   n1 = SetDateText!() -> (then, DateTime:Date Time Structure) {is=title K2Node_FunctionEntry}
   > SetDateText!(
    execute
    self:WBP Community Date Object Reference=Get StartDate() -> (StartDate:WBP Community Date Object Reference)
    DateTime:Date Time Structure=@n1
   ) -> (then) {is=title K2Node_CallFunction}
  ]}
 SetHighlightText {nodes [
   n1 = SetHighlightText!() -> (then, Text:Text) {is=title K2Node_FunctionEntry}
   > SetText!(execute, self:WBP UIKit Highlight Base Object Reference=Get Highlight(), Text:Text=@n1) -> (then)
   > SetFontSize(self=Get Text_Highlight(self=Get Highlight()), DisplayFontSize=16)
  ]}
]
variables []
widgets {
 info {inheritedWidgetCount=0, parentClass=&UserWidget, rootWidgetClass=&GridPanel, widgetCount=15}
 widgets [
  GridPanel_36 {widgetClassPath=&GridPanel, widget=GridPanel_36, properties {columnFill [0,1], rowFill [1,1]}}
  StartDate {
   widgetClassPath=&/WBP_Community_Date
   bIsVariable=true
   parent=GridPanel_36
   widget=StartDate
   properties {bOverride_Cursor=true, size=<88 88>}
   slot {class=GridSlot, properties {rowSpan=2}}
  }
  StateSwitcher {
   widgetClassPath=&CommonVisibilitySwitcher
   bIsVariable=true
   parent=GridPanel_36
   widget=StateSwitcher
   slot {class=GridSlot, properties {column=1, rowSpan=2, padding=<ltrb 24 0 0 0>}}
  }
  VB_Upcoming {
   widgetClassPath=&VerticalBox
   bIsVariable=true
   parent=StateSwitcher
   widget=VB_Upcoming
   slot {class=CommonVisibilitySwitcherSlot, properties {verticalAlignment=VAlign_Center}}
  }
  Text_EventUpcomingTime {
   widgetClassPath=&CommonTextBlock
   bIsVariable=true
   parent=VB_Upcoming
   widget=Text_EventUpcomingTime
   properties {
    bIsScrollingEnabled=false
    text="{Time} · For {## Hours}"
    font {
     bForceMonospaced=false
     bMaterialIsStencil=false
     fontMaterial=null
     fontObject=&NotoSans
     letterSpacing=0
     monospacedWidth=1
     size=24
     skewAmount=0
     typefaceFontName=NotoSans-Regular
     outlineSettings {
      bApplyOutlineToDropShadows=false
      bMiteredCorners=false
      bSeparateFillAlpha=false
      outlineMaterial=null
      outlineSize=0
      outlineColor=<rgba 0 0 0 1>
     }
    }
    shadowOffset=<0 0>
   }
   slot {class=VerticalBoxSlot, properties {padding=<ltrb 0 0 0 3>}}
  }
  Text_CommunityName {
   widgetClassPath=&CommonTextBlock
   bIsVariable=true
   parent=VB_Upcoming
   widget=Text_CommunityName
   properties {
    bIsScrollingEnabled=false
    mobileFontSizeMultiplier=1.1
    text="{CommunityName}"
    colorAndOpacity {colorUseRule=UseColor_Specified, specifiedColor=<rgba 1 1 1 0.651>}
    font {
     bForceMonospaced=false
     bMaterialIsStencil=false
     fontMaterial=null
     fontObject=&HeadingNow
     letterSpacing=0
     monospacedWidth=1
     size=14.25
     skewAmount=0
     typefaceFontName=74 Regular
     outlineSettings {
      bApplyOutlineToDropShadows=false
      bMiteredCorners=false
      bSeparateFillAlpha=false
      outlineMaterial=null
      outlineSize=0
      outlineColor=<rgba 0 0 0 1>
     }
    }
    shadowOffset=<0 0>
   }
   slot {class=VerticalBoxSlot, properties {padding=<ltrb 0 3 0 0>}}
  }
  VB_Live {
   widgetClassPath=&VerticalBox
   bIsVariable=true
   parent=StateSwitcher
   widget=VB_Live
   slot {class=CommonVisibilitySwitcherSlot, properties {verticalAlignment=VAlign_Center}}
  }
  Highlight {
   widgetClassPath=/FortUILibrary/UIKit/Highlight/WBP_UIKit_Highlight
   bIsVariable=true
   parent=VB_Live
   widget=Highlight
   properties {highlightText="live now  · ends in {time}", highlightType=LiveAlt}
   slot {class=VerticalBoxSlot, properties {horizontalAlignment=HAlign_Left, padding=<ltrb 0 0 0 3>}}
  }
  Text_EventStartedAgo {
   widgetClassPath=&CommonTextBlock
   bIsVariable=true
   parent=VB_Live
   widget=Text_EventStartedAgo
   properties {
    bIsScrollingEnabled=false
    mobileFontSizeMultiplier=1.1
    text="Started {X} ago"
    font {
     bForceMonospaced=false
     bMaterialIsStencil=false
     fontMaterial=null
     fontObject=&HeadingNow
     letterSpacing=0
     monospacedWidth=1
     size=14.25
     skewAmount=0
     typefaceFontName=74 Regular
     outlineSettings {
      bApplyOutlineToDropShadows=false
      bMiteredCorners=false
      bSeparateFillAlpha=false
      outlineMaterial=null
      outlineSize=0
      outlineColor=<rgba 0 0 0 1>
     }
    }
    shadowOffset=<0 0>
   }
   slot {class=VerticalBoxSlot, properties {padding=<ltrb 0 3 0 0>}}
  }
  VB_Ended {
   widgetClassPath=&VerticalBox
   bIsVariable=true
   parent=StateSwitcher
   widget=VB_Ended
   slot {class=CommonVisibilitySwitcherSlot, properties {verticalAlignment=VAlign_Center}}
  }
  Text_EventEndedTime {
   widgetClassPath=&CommonTextBlock
   bIsVariable=true
   parent=VB_Ended
   widget=Text_EventEndedTime
   properties {
    bIsScrollingEnabled=false
    text="Ended {##} ago"
    font {
     bForceMonospaced=false
     bMaterialIsStencil=false
     fontMaterial=null
     fontObject=&NotoSans
     letterSpacing=0
     monospacedWidth=1
     size=24
     skewAmount=0
     typefaceFontName=NotoSans-Regular
     outlineSettings {
      bApplyOutlineToDropShadows=false
      bMiteredCorners=false
      bSeparateFillAlpha=false
      outlineMaterial=null
      outlineSize=0
      outlineColor=<rgba 0 0 0 1>
     }
    }
    shadowOffset=<0 0>
   }
   slot {class=VerticalBoxSlot, properties {padding=<ltrb 0 0 0 3>}}
  }
  Text_CommunityName_1 {
   widgetClassPath=&CommonTextBlock
   bIsVariable=true
   parent=VB_Ended
   widget=Text_CommunityName_1
   properties {
    bIsScrollingEnabled=false
    mobileFontSizeMultiplier=1.1
    text="{CommunityName}"
    colorAndOpacity {colorUseRule=UseColor_Specified, specifiedColor=<rgba 1 1 1 0.651>}
    font {
     bForceMonospaced=false
     bMaterialIsStencil=false
     fontMaterial=null
     fontObject=&HeadingNow
     letterSpacing=0
     monospacedWidth=1
     size=14.25
     skewAmount=0
     typefaceFontName=74 Regular
     outlineSettings {
      bApplyOutlineToDropShadows=false
      bMiteredCorners=false
      bSeparateFillAlpha=false
      outlineMaterial=null
      outlineSize=0
      outlineColor=<rgba 0 0 0 1>
     }
    }
    shadowOffset=<0 0>
   }
   slot {class=VerticalBoxSlot, properties {padding=<ltrb 0 3 0 0>}}
  }
  VB_Canceled {
   widgetClassPath=&VerticalBox
   bIsVariable=true
   parent=StateSwitcher
   widget=VB_Canceled
   slot {class=CommonVisibilitySwitcherSlot, properties {verticalAlignment=VAlign_Center}}
  }
  Text_EventCanceledTime {
   widgetClassPath=&CommonTextBlock
   bIsVariable=true
   parent=VB_Canceled
   widget=Text_EventCanceledTime
   properties {
    bIsScrollingEnabled=false
    text=Canceled
    font {
     bForceMonospaced=false
     bMaterialIsStencil=false
     fontMaterial=null
     fontObject=&NotoSans
     letterSpacing=0
     monospacedWidth=1
     size=24
     skewAmount=0
     typefaceFontName=NotoSans-Regular
     outlineSettings {
      bApplyOutlineToDropShadows=false
      bMiteredCorners=false
      bSeparateFillAlpha=false
      outlineMaterial=null
      outlineSize=0
      outlineColor=<rgba 0 0 0 1>
     }
    }
    shadowOffset=<0 0>
   }
   slot {class=VerticalBoxSlot, properties {padding=<ltrb 0 0 0 3>}}
  }
  Text_CommunityName_2 {
   widgetClassPath=&CommonTextBlock
   bIsVariable=true
   parent=VB_Canceled
   widget=Text_CommunityName_2
   properties {
    bIsScrollingEnabled=false
    mobileFontSizeMultiplier=1.1
    text="{CommunityName}"
    colorAndOpacity {colorUseRule=UseColor_Specified, specifiedColor=<rgba 1 1 1 0.651>}
    font {
     bForceMonospaced=false
     bMaterialIsStencil=false
     fontMaterial=null
     fontObject=&HeadingNow
     letterSpacing=0
     monospacedWidth=1
     size=14.25
     skewAmount=0
     typefaceFontName=74 Regular
     outlineSettings {
      bApplyOutlineToDropShadows=false
      bMiteredCorners=false
      bSeparateFillAlpha=false
      outlineMaterial=null
      outlineSize=0
      outlineColor=<rgba 0 0 0 1>
     }
    }
    shadowOffset=<0 0>
   }
   slot {class=VerticalBoxSlot, properties {padding=<ltrb 0 3 0 0>}}
  }
 ]
}
