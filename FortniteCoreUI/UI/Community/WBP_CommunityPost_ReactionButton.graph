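# Reaction button for community posts, extending the UIKit generic modular button with reacted-state visuals and a reaction count.
# Hover plays a bounce, press and release play a squash forward and in reverse, and a click while HasReacted is false plays a material-driven confetti burst.
# OnHasReactedChanged drives the background material's Display Regular parameter.
# SetReactionCount shows shorthand count text only above zero.
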WidgetBlueprint /FortniteCoreUI/UI/Community/WBP_CommunityPost_ReactionButton
animations [
 Anim_Bounce {
  binding_count=1
  length_seconds=0.2961
  object=&self:Anim_Bounce
  sequence {
   bindings [Main {
     id=#1
     possessed_class=GridPanel
     tracks [{
       type=MovieScene2DTransformTrack
       sections [{
         type=MovieScene2DTransformSection
         channels [
          Translation.X {type=MovieSceneScriptingFloatChannel}
          Translation.Y {type=MovieSceneScriptingFloatChannel, keys [[0,0],[4.423,-5],[13.268,0],[17.765,0]]}
         ]
        }]
      }]
    }]
   display_rate {numerator=60}
  }
 }
 Anim_Pressed {
  binding_count=2
  length_seconds=0.1
  object=&self:Anim_Pressed
  sequence {
   bindings [
    Root {
     id=#2
     possessed_class=GridPanel
     tracks [{
       type=MovieScene2DTransformTrack
       sections [{
         type=MovieScene2DTransformSection
         end_frame=6
         start_frame=0
         channels [
          Scale.X {type=MovieSceneScriptingFloatChannel, keys [[0,1],[6,0.975]]}
          Scale.Y {type=MovieSceneScriptingFloatChannel, keys [[0,1],[6,0.975]]}
         ]
        }]
      }]
    }
    Main {
     id=#3
     possessed_class=GridPanel
     tracks [{
       type=MovieScene2DTransformTrack
       sections [{
         type=MovieScene2DTransformSection
         channels [
          Scale.X {type=MovieSceneScriptingFloatChannel, keys [[0,1],[6,0.95]]}
          Scale.Y {type=MovieSceneScriptingFloatChannel, keys [[0,1],[6,0.975]]}
         ]
        }]
      }]
    }
   ]
   display_rate {numerator=60}
  }
 }
 Anim_Reaction {
  binding_count=1
  length_seconds=0.8537
  object=&self:Anim_Reaction
  sequence {bindings [Image_Confetti {
     id=#4
     possessed_class=Image
     tracks [
      {
       type=MovieSceneWidgetMaterialTrack
       sections [{
         type=MovieSceneParameterSection
         end_frame=17
         start_frame=-1
         channels [
          Reveal {type=MovieSceneScriptingFloatChannel, keys [[0,0],[17.001,0.36]]}
          SegmentSize {type=MovieSceneScriptingFloatChannel, default=0.9, keys [[-0.0367,0.88],[4,0.9046],[17.0047,0.9]]}
         ]
        }]
      }
      {
       type=MovieSceneByteTrack
       display_name=Visibility
       sections [{type=MovieSceneByteSection, end_frame=17, start_frame=0, channels [None {type=MovieSceneScriptingByteChannel, keys [[0,4],[17.001,2]]}]}]
      }
      {
       type=MovieScene2DTransformTrack
       sections [{
         type=MovieScene2DTransformSection
         end_frame=17
         start_frame=-1
         channels [
          Translation.X {type=MovieSceneScriptingFloatChannel, const [0,1]}
          Translation.Y {type=MovieSceneScriptingFloatChannel, const [0,1]}
          Angle {type=MovieSceneScriptingFloatChannel, const [0,1]}
          Scale.X {type=MovieSceneScriptingFloatChannel, keys [[-0.0207,0.8],[17.0367,1.3]]}
          Scale.Y {type=MovieSceneScriptingFloatChannel, keys [[-0.0207,0.8],[17.0367,1.3]]}
          Shear.X {type=MovieSceneScriptingFloatChannel, const [0,1]}
          Shear.Y {type=MovieSceneScriptingFloatChannel, const [0,1]}
         ]
        }]
      }
     ]
    }]}
 }
]
graphs [
 EventGraph {nodes [
   Event On Clicked()
   > Parent: On Clicked() {is=title K2Node_CallParentFunction}
   > n1 = Branch(Condition=Get HasReacted() -> (HasReacted:bool))
   PlayAnimation(execute=@n1.else, InAnimation=Get Anim_Reaction() -> (Anim_Reaction:Widget Animation Object Reference))
   Event On Hovered()
   > Parent: On Hovered() {is=title K2Node_CallParentFunction}
   > PlayAnimationForward(InAnimation=Get Anim_Bounce() -> (Anim_Bounce:Widget Animation Object Reference))
   Event On Pressed() {is=title K2Node_Event}
   > Parent: On Pressed() {is=title K2Node_CallParentFunction}
   n2 = Get Anim_Pressed() -> (Anim_Pressed:Widget Animation Object Reference)
   > PlayAnimationForward(InAnimation=@n2)
   Event On Released() {is=title K2Node_Event}
   > Parent: On Released() {is=title K2Node_CallParentFunction}
   > PlayAnimationReverse(InAnimation=@n2)
  ]}
 OnHasReactedChanged {nodes [
   n1 = OnHasReactedChanged!() -> (then, InHasReacted:bool=~) {is=title K2Node_FunctionEntry}
   > n2 = GetBackgroundDynamicMaterial!(execute, self:WBP UIKit Button Generic Base Object Reference) -> (then, Dynamic Material:Material Instance Dynamic Object Reference)
   > Is Valid(InputObject=@n2)
   > SetScalarParameterValue(self=@n2, ParameterName=Display Regular, Value="To Float (Boolean)"(InBool=@n1))
  ]}
 SetReactionCount {nodes [
   n1 = SetReactionCount!() -> (then, Count:int) {is=title K2Node_FunctionEntry}
   n2 = "integer > integer"(A=@n1)
   > SetBlockVisibility!(execute, self:WBP UIKit Block Base Object Reference=Get Block_Text() -> (Block_Text:WBP UIKit Block Text Object Reference), Visible:bool=@n2) -> (then)
   > Branch(Condition=@n2)
   > SetText!(
    execute
    self:WBP UIKit Button Generic Base Object Reference
    Text:Text=ConvertQuantityToShorthandText!(Quantity:int=Get Count() -> (Count:int=0)) -> (Text:Text)
   ) -> (then)
  ]}
]
variables [HasReacted {type=boolean}]
widgets {
 info {
  inheritedWidgetCount=0
  namedSlotCount=1
  parentClass=/FortUILibrary/Systems/ModularButton/WBP_UIKit_ButtonGeneric_Base.WBP_UIKit_ButtonGeneric_Base_C
  rootWidgetClass=null
  widgetCount=4
 }
 widgets [
  Overlay_0 {widgetClassPath=&Overlay, widget=Overlay_0}
  SizeBox_1 {
   widgetClassPath=&SizeBox
   parent=Overlay_0
   widget=SizeBox_1
   slot {class=OverlaySlot, properties {horizontalAlignment=HAlign_Fill, verticalAlignment=VAlign_Center, padding=<ltrb 46 0 0 0>}}
  }
  CommonVisualAttachment_74 {
   widgetClassPath=&CommonVisualAttachment
   parent=SizeBox_1
   widget=CommonVisualAttachment_74
   properties {visibility=HitTestInvisible, widthOverride=160, contentAnchor=<0.5 0.5>}
   slot {class=SizeBoxSlot, properties {horizontalAlignment=HAlign_Center, verticalAlignment=VAlign_Center}}
  }
  Image_Confetti {
   widgetClassPath=&Image
   bIsVariable=true
   parent=CommonVisualAttachment_74
   widget=Image_Confetti
   properties {
    visibility=Hidden
    brush {
     bIsDynamicallyLoaded=false
     drawAs=Image
     imageType=FullColor
     mirroring=NoMirror
     resourceName=None
     resourceObject=/FortniteCoreUI/UI/Community/Art/Materials/MI_UI_Community_Burst
     tiling=NoTile
     imageSize=<150 150>
     margin=<ltrb 0 0 0 0>
     outlineSettings {
      bUseBrushTransparency=false
      roundingType=HalfHeightRadius
      width=0
      color {colorUseRule=UseColor_Specified, specifiedColor=<rgba 0 0 0 0>}
      cornerRadii=<0 0 0 0>
     }
     tintColor {colorUseRule=UseColor_Specified, specifiedColor=<rgba 1 1 1 1>}
     uVRegion {bIsValid=false, max=<0 0>, min=<0 0>}
    }
   }
   slot {class=SizeBoxSlot, properties {verticalAlignment=VAlign_Center}}
  }
 ]
}
