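# Community event "Interested" button widget, derived from the UIKit generic button, that adapts to an event's status.
# For Live events it hides the icon and reads "Jump In"; otherwise it shows the icon and reads "Interested".
# A separate function collapses it for Ended or Canceled events, and another swaps the background material between loud and quiet styles.
# PreConstruct and Construct only call the parent; state changes come through these functions.
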WidgetBlueprint /FortniteCoreUI/UI/Community/WBP_Community_InterestedButton
graphs [
 EventGraph {nodes [
   n1 = Event PreConstruct()
   > Parent: PreConstruct(IsDesignTime=@n1.IsDesignTime)
   Event Construct()
   > Parent: Construct()
  ]}
 SetInterestedButtonState {nodes [
   n1 = SetInterestedButtonState!() -> (then, EventStatus:ECommunityEventStatus Enum=Unknown) {is=title K2Node_FunctionEntry}
   n2 = "Equal (Enum)"(A:ECommunityEventStatus Enum=@n1, B:ECommunityEventStatus Enum=Live)
   > SetBlockVisibility!(
    execute
    self:WBP UIKit Block Base Object Reference=Get Block_Icon() -> (Block_Icon:WBP UIKit Block Image Object Reference)
    Visible:bool=NOT Boolean(A=@n2)
   ) -> (then)
   > SetText!(
    execute
    self:WBP UIKit Button Generic Base Object Reference
    Text:Text=SelectText(A="LOCTEXT(\"Jump In\")", B="LOCTEXT(\"Interested\")", bPickA=@n2)
   ) -> (then)
  ]}
 SetInterestedButtonVisibility {nodes [
   n1 = SetInterestedButtonVisibility!() -> (then, EventStatus:ECommunityEventStatus Enum=Unknown) {is=title K2Node_FunctionEntry}
   > SetVisibility(InVisibility=Select!(
     Unknown:ESlateVisibility Enum=SelfHitTestInvisible
     Upcoming:ESlateVisibility Enum=SelfHitTestInvisible
     Live:ESlateVisibility Enum=SelfHitTestInvisible
     Ended:ESlateVisibility Enum=Collapsed
     Canceled:ESlateVisibility Enum=Collapsed
     Index:ECommunityEventStatus Enum=@n1
    ) -> (ReturnValue))
  ]}
 SetIntestedButtonType {nodes [
   n1 = SetIntestedButtonType!() -> (then, IsLoud:bool=~) {is=title K2Node_FunctionEntry}
   > SetBackgroundMaterials!(
    execute
    self:WBP UIKit Button Generic Base Object Reference
    Background MI:Material Instance Object Reference=SelectMaterialInstance!(
     A:Material Instance Object Reference=/FortUILibrary/UIKit/Buttons/MaterialInstances/CTAButtons/MI_UIKit_BgLoud
     B:Material Instance Object Reference=/FortUILibrary/UIKit/Buttons/MaterialInstances/CTAButtons/MI_UIKit_BgQuiet
     bSelectA:bool=@n1
    ) -> (ReturnValue:Material Instance Object Reference)
    Outline MI:Material Instance Object Reference=/FortUILibrary/UIKit/Buttons/MaterialInstances/TileOutline/MI_UIKit_Outline
   ) -> (then)
  ]}
]
variables []
widgets {
 info {
  inheritedWidgetCount=0
  parentClass=/FortUILibrary/UIKit/Buttons/WBP_UIKit_Button_Generic.WBP_UIKit_Button_Generic_C
  rootWidgetClass=null
  widgetCount=0
 }
 widgets []
}
