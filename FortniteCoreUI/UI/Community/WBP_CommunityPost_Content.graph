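# WBP_CommunityPost_Content is the detail view for a community post, showing either highlight or event content, plus loading and failure states.
# A CommunityPostVM handles reactions, replies, interest toggling and reporting, with SetHighlightId creating a CommunityHighlightVM.
# Helper functions format time and status text, load banner images with error fallback, and render QR-code textures.
# MOTD buttons spawn dynamically for calls to action, with custom navigation and last-focus restoration.
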WidgetBlueprint /FortniteCoreUI/UI/Community/WBP_CommunityPost_Content
graphs [
 CF_CreatedDateTimeToText {nodes [
   n1 = CF_CreatedDateTimeToText!() -> (then, DateTime:Date Time Structure) {is=title K2Node_FunctionEntry}
   > n2 = ConvertDateTimeToTimeAgoText!(execute, DateTime:Date Time Structure=@n1) -> (then, Text:Text)
   > Return Node!(execute, Text:Text=@n2)
  ]}
 CF_EventDateTimeEndedText {nodes [
   CF_EventDateTimeEndedText!() -> (then, EndTime:Date Time Structure) {is=title K2Node_FunctionEntry}
   > n1 = ConvertTimespanToTimerText!(
    execute
    self:WBP Community Post Date Info Object Reference=Get WBP_CommunityPost_DateInfo() -> (WBP_CommunityPost_DateInfo:WBP Community Post Date Info Object Reference)
    Timespan:Timespan Structure=DateTime - DateTime!(A:Date Time Structure=UTC Now!() -> (ReturnValue:Date Time Structure), B:Date Time Structure=Get EndTime() -> (EndTime:Date Time Structure)) -> (ReturnValue:Timespan Structure)
   ) -> (then, Text:Text)
   > Return Node(Text=Format Text!(Format="LOCTEXT(\"Ended {##} ago\")", "##":Text=@n1) -> (Result))
  ]}
 CF_EventDateTimeToEventStartedAgoText {nodes [
   CF_EventDateTimeToEventStartedAgoText!() -> (then, StartTime:Date Time Structure) {is=title K2Node_FunctionEntry}
   > n1 = ConvertTimespanToTimerText(self=Get WBP_CommunityPost_DateInfo(), Timespan=DateTime - DateTime(A=UTC Now(), B=Get StartTime() -> (StartTime:Date Time Structure)))
   > Return Node(Text=Format Text(Format="LOCTEXT(\"Started {##} ago\")", "##"=@n1))
  ]}
 CF_EventDateTimeToHighlightText {nodes [
   n1 = CF_EventDateTimeToHighlightText!() -> (then, EndTime:Date Time Structure) {is=title K2Node_FunctionEntry}
   > n2 = ConvertTimespanToTimerText(self=Get WBP_CommunityPost_DateInfo(), Timespan=DateTime - DateTime(A=@n1, B=UTC Now()))
   > Return Node(Text=Format Text(Format="LOCTEXT(\"LIVE NOW  · ENDS IN {##}\")", "##"=@n2))
  ]}
 CF_GetDateSideColor {nodes [
   n1 = CF_GetDateSideColor!() -> (then, IsInterested:bool=~, EventStatus:ECommunityEventStatus Enum=Unknown) {is=title K2Node_FunctionEntry}
   n2 = Get WBP_CommunityPost_DateInfo()
   > Return Node!(
    execute
    Color:Linear Color Structure=GetDateSideColor!(
     self:WBP Community Date Object Reference=Get StartDate!(self:WBP Community Post Date Info Object Reference=@n2) -> (StartDate:WBP Community Date Object Reference)
     IsInterested:bool=@n1
     EventStatus:ECommunityEventStatus Enum=@n1.EventStatus
    ) -> (ReturnValue:Linear Color Structure)
   )
  ]}
 CF_GetEventInterestedCCUVisibility {nodes [
   n1 = CF_GetEventInterestedCCUVisibility!() -> (then, InterestedCount:int) {is=title K2Node_FunctionEntry}
   > Return Node!(execute, Visibility:ESlateVisibility Enum=Select(Option 0=Collapsed, Option 1=SelfHitTestInvisible, Index="integer > integer"(A=@n1)))
  ]}
 CF_GetInterestedButtonType {nodes [
   n1 = CF_GetInterestedButtonType!() -> (then, EventStatus:ECommunityEventStatus Enum=Unknown, IsInterested:bool=~) {is=title K2Node_FunctionEntry}
   > Return Node!(
    execute
    IsLoud:bool=OR Boolean(A="Equal (Enum)"(A:ECommunityEventStatus Enum=@n1, B:ECommunityEventStatus Enum=Live), B=NOT Boolean(A=@n1.IsInterested))
   )
  ]}
 ConvertDetailsText {nodes [
   n1 = ConvertDetailsText!() -> (then, DateTime:Date Time Structure, CommunityName:Text) {is=title K2Node_FunctionEntry}
   > Return Node!(
    execute
    DetailsText:Text=Format Text!(
     Format="LOCTEXT(\"Posted {Date} in the {CommunityName} Community\")"
     Date:Text=ConvertDateTimeToTimeAgoText!(DateTime=@n1) -> (Text)
     CommunityName:Text=@n1.CommunityName
    ) -> (Result)
   )
  ]}
 ConvertViewRepliesText {nodes [
   n1 = ConvertViewRepliesText!() -> (then, NumReplies:int) {is=title K2Node_FunctionEntry}
   > n2 = Branch(Condition="Equal (Integer)"(A=@n1))
   > Return Node!(execute, Text:Text="LOCTEXT(\"View on Web\")")
   n3 = Get NumReplies() -> (NumReplies:int=0)
   Return Node(
    execute=@n2.else
    Text=Format Text!(
     Format="LOCTEXT(\"View {DisplayNumber} {RawNumber}|plural(one=Reply,other=Replies)\")"
     DisplayNumber:Text=ConvertQuantityToShorthandText!(Quantity:int=@n3) -> (Text:Text)
     RawNumber:int=@n3
    ) -> (Result)
   )
  ]}
 EventGraph {nodes [
   n1 = "On Clicked (Button_Reaction)"!() -> (OutputDelegate, then, Button:Common Button Base Object Reference) {is=title K2Node_ComponentBoundEvent}
   > n2 = Set LastFocusedWidget(LastFocusedWidget:Widget Object Reference=@n1.Button) -> (Output_Get:Widget Object Reference)
   ToggleReaction(
    execute=[@n2.then, @n5.then]
    self:Community Post VM Object Reference=Get CommunityPostVM() -> (CommunityPostVM:Community Post VM Object Reference)
   )
   n3 = "On Clicked (Button_ViewReplies)"!() -> (OutputDelegate, then, Button:Common Button Base Object Reference) {is=title K2Node_ComponentBoundEvent}
   > n4 = Set LastFocusedWidget(LastFocusedWidget=@n3.Button)
   ViewReplies(execute=[@n4.then, @n6.then], self:Community Post VM Object Reference=Get CommunityPostVM())
   "On Loading Errored (Image_Texture)"!() -> (OutputDelegate, then, ErrorCode:int=0, ModuleId:int=0, ErrorString:String, ErrorMessage:Text) {is=title K2Node_ComponentBoundEvent}
   > SetThumbnailImageFallback()
   "On Loading Errored (Image_Texture_Event)"!() -> (OutputDelegate, then, ErrorCode:int=0, ModuleId:int=0, ErrorString:String, ErrorMessage:Text) {is=title K2Node_ComponentBoundEvent}
   > SetEventThumbnailImageFallback()
   "On Clicked (Button_Interested)"!() -> (OutputDelegate, then, Button:Common Button Base Object Reference) {is=title K2Node_ComponentBoundEvent}
   > OnInterestedButtonClicked()
   n5 = "On Clicked (Button_EventReaction)"!() -> (OutputDelegate, then, Button:Common Button Base Object Reference) {is=title K2Node_ComponentBoundEvent}
   n6 = "On Clicked (Button_EventReplies)"!() -> (OutputDelegate, then, Button:Common Button Base Object Reference) {is=title K2Node_ComponentBoundEvent}
   n7 = OnButtonEntryClicked!() -> (OutputDelegate, then, Button:Common Button Base Object Reference) {is=title K2Node_CustomEvent}
   > Set LastFocusedWidget(LastFocusedWidget=@n7.Button)
   n8 = Find Item!(
    TargetArray:Array of User Widget Object References=GetAllEntries!(self:Dynamic Entry Box Base Object Reference=Get EntryBox_Buttons() -> (EntryBox_Buttons:Fort Dynamic Entry Box Object Reference)) -> (ReturnValue:Array of User Widget Object References)
    ItemToFind:"User Widget Object Reference (by ref)"=@n7.Button
   ) -> (ReturnValue)
   > Branch(Condition=Is Valid Index!(TargetArray:Array of Motd Button Structures=Get MotdButtons() -> (MotdButtons:Array of Motd Button Structures), IndexToTest:int=@n8) -> (ReturnValue:bool))
   n9 = Get MotdButtons()
   > CallToAction!(
    execute
    self:Motd Action Base Object Reference=Break Motd Button!(MotdButton:"Motd Button Structure (by ref)"="Get (a ref)"(Array:Array of Motd Button Structures=@n9, Dimension 1=@n8) -> (Output:"Motd Button Structure (by ref)")) -> (Text:Text, Style:EMotdButtonStyle Enum, Action:Motd Action Base Object Reference) {is=title K2Node_BreakStruct}.Action
    PlayerController:Fort Player Controller Object Reference=GetFirstLocalFortPlayerController!(bIncludeNonSpectators:bool=true, bIncludeSpectators:bool=false) -> (ReturnValue:Fort Player Controller Object Reference)
   ) -> (then)
   > Call On CTAButton Clicked!(execute:Exec, self:WBP Community Post Content Object Reference) -> (then:Exec) {is=title K2Node_CallDelegate}
   Event PreConstruct()
   > n10 = Get!(
    execute:Exec
    self:WBP UIKit Button CTA Base Object Reference=Get Button_ViewReplies() -> (Button_ViewReplies:WBP UIKit Button Regular Object Reference)
   ) -> (then:Exec, else:Exec, Block_Icon:WBP UIKit Block Image Object Reference)
   > SetGridSlotPadding!(execute, Widget:Widget Object Reference=@n10, Padding:Margin Structure) -> (then)
   Event Construct()
   > SetCompactUIStyle() {is=title K2Node_CallFunction}
   > SetNavRules()
   > n11 = GetDynamicMaterial(self=Get Image_Blur() -> (Image_Blur:Image Object Reference))
   n12 = GetConfigDataBool!(EntryName:String=EnableBlurForFrontEndWidgets) -> (bValueOut:bool, ReturnValue:bool)
   > SetScalarParameterValue(
    self=@n11
    ParameterName=UseBlur
    Value="To Float (Boolean)"(InBool=Select(Option 0:bool=true, Option 1:bool=@n12, Index=@n12.ReturnValue) -> (ReturnValue:bool))
   )
  ]}
 GetFocusTarget {nodes [
   GetFocusTarget() {is=title K2Node_FunctionEntry}
   n1 = IsEvent!(self:Community Post VM Object Reference=Get CommunityPostVM()) -> (ReturnValue:bool)
   n2 = Get IsEvent() -> (IsEvent:bool)
   > n3 = Branch(Condition=NOT Boolean(A=OR Boolean(A=@n1, B=@n2)))
   n4 = GetAllEntries(self=Get EntryBox_Buttons())
   > n5 = Branch(Condition=Is Not Empty!(TargetArray:Array of User Widget Object References=@n4) -> (ReturnValue:bool))
   > n6 = Get!(execute) -> (then, else, LastFocusedWidget:Widget Object Reference)
   > Return Node!(execute, TargetWidget:Widget Object Reference=@n6)
   Return Node(execute=@n6.else, TargetWidget="Get (a copy)"(Array:Array of User Widget Object References=@n4) -> (Output:User Widget Object Reference))
   Return Node(execute=@n5.else, TargetWidget=Get Button_Reaction() -> (Button_Reaction:WBP Community Post Reaction Button Object Reference))
   n7 = Get Button_Interested() -> (Button_Interested:WBP Community Interested Button Object Reference)
   Return Node(
    execute=@n3.else
    TargetWidget=Select(
     Option 0:Widget Object Reference=Get Button_EventReaction() -> (Button_EventReaction:WBP Community Post Reaction Button Object Reference)
     Option 1:Widget Object Reference=@n7
     Index=IsVisible(self=@n7)
    ) -> (ReturnValue:Widget Object Reference)
   )
  ]}
 GetIsCreatorEvent {nodes [
   GetIsCreatorEvent() {is=title K2Node_FunctionEntry}
   n1 = Get CommunityPostVM()
   > Return Node!(execute, ReturnNode:bool=OR Boolean(A=IsEvent(self=@n1), B=Get IsEvent()))
  ]}
 GetQRCodeVisibility {nodes [
   n1 = GetQRCodeVisibility!() -> (then, CanViewReplies:bool=~) {is=title K2Node_FunctionEntry}
   n2 = SupportsQRCode!() -> (ReturnValue:bool)
   > Return Node!(
    execute
    Visibility:ESlateVisibility Enum=Select(Option 0:ESlateVisibility Enum=Collapsed, Option 1:ESlateVisibility Enum=SelfHitTestInvisible, Index=AND Boolean(A=@n1, B=@n2)) -> (ReturnValue:ESlateVisibility Enum)
   )
  ]}
 NavigateExternalDirection {nodes [
   n1 = NavigateExternalDirection!() -> (then, Navigation:EUINavigation Enum) {is=title K2Node_FunctionEntry}
   > Call On Navigate!(execute:Exec, self:WBP Community Post Content Object Reference, Navigation:EUINavigation Enum=@n1) -> (then:Exec) {is=title K2Node_CallDelegate}
   > Return Node!(execute, ReturnValue:Widget Object Reference)
  ]}
 NavigateUpToButtonList {nodes [
   NavigateUpToButtonList!() -> (then, Navigation:EUINavigation Enum) {is=title K2Node_FunctionEntry}
   n1 = GetAllEntries(self=Get EntryBox_Buttons())
   > Return Node(ReturnValue="Get (a copy)"(Array=@n1, Dimension 1=Last Index!(TargetArray:Array of User Widget Object References=@n1) -> (ReturnValue=0)))
  ]}
 OnActivated {nodes [
   OnActivated() {is=title K2Node_FunctionEntry}
   > SetScrollWidget!(
    execute
    ToScroll:Widget Object Reference=Select(
     Option 0:Widget Object Reference=Get ScrollBox() -> (ScrollBox:Scroll Box Object Reference)
     Option 1:Widget Object Reference=Get ScrollBox_Event() -> (ScrollBox_Event:Scroll Box Object Reference)
     Index=GetIsCreatorEvent!() -> (ReturnNode:bool) {is=title K2Node_CallFunction}
    ) -> (ReturnValue:Widget Object Reference)
   ) -> (then)
  ]}
 OnCTAButtonClicked {nodes [OnCTAButtonClicked() {is=title K2Node_FunctionEntry}]}
 OnInterestedButtonClicked {nodes [
   OnInterestedButtonClicked() {is=title K2Node_FunctionEntry}
   n1 = Get CommunityPostVM()
   > n2 = Branch(Condition="Equal (Enum)"(A=GetEventStatus!(self:Community Post VM Object Reference=@n1) -> (ReturnValue:ECommunityEventStatus Enum=Unknown), B=Live))
   n3 = Get ActionButtons!(self:Community Post VM Object Reference=Get CommunityPostVM()) -> (ActionButtons:Array of Motd Button Structures)
   > Branch(Condition=Is Not Empty(TargetArray:Array of Motd Button Structures=@n3))
   > CallToAction(
    self=Break Motd Button(MotdButton="Get (a copy)"(Array:Array of Motd Button Structures=@n3) -> (Output:Motd Button Structure)).Action
    PlayerController=GetFirstLocalFortPlayerController(bIncludeNonSpectators=true, bIncludeSpectators=false)
   )
   > Call On CTAButton Clicked()
   ToggleInterested(execute=@n2.else, self:Community Post VM Object Reference=Get CommunityPostVM())
  ]}
 OnNavigate {nodes [OnNavigate!() -> (then, Navigation:EUINavigation Enum) {is=title K2Node_FunctionEntry}]}
 OnSuccessFinishedQuerying {nodes [OnSuccessFinishedQuerying() {is=title K2Node_FunctionEntry}]}
 Report {nodes [
   Report() {is=title K2Node_FunctionEntry}
   n1 = Get CommunityPostVM()
   > Is Valid(InputObject=@n1)
   > Report(self:Community Post VM Object Reference=@n1) {is=title K2Node_CallFunction}
  ]}
 SetButtons {nodes [
   n1 = SetButtons!() -> (then, Buttons:Array of Motd Button Structures) {is=title K2Node_FunctionEntry}
   > Reset!(execute, self:Dynamic Entry Box Object Reference=Get EntryBox_Buttons(), bDeleteWidgets:bool=false) -> (then)
   > Set MotdButtons(MotdButtons:Array of Motd Button Structures=@n1) -> (Output_Get:Array of Motd Button Structures)
   > n2 = For Each Loop(Array:Array of Motd Button Structures=@n1) -> (Array Element:Motd Button Structure)
   n3 = Break Motd Button(MotdButton=@n2)
   > n4 = Create Entry of Class!(
    execute
    self:Dynamic Entry Box Object Reference=Get EntryBox_Buttons()
    EntryClass:User Widget Class Reference=Select!(
     Main:User Widget Class Reference=/FortUILibrary/UIKit/Buttons/WBP_UIKit_Button_Loud.WBP_UIKit_Button_Loud_C
     Secondary:User Widget Class Reference=/FortUILibrary/UIKit/Buttons/WBP_UIKit_Button_Regular.WBP_UIKit_Button_Regular_C
     None:User Widget Class Reference
     Index:EMotdButtonStyle Enum=@n3.Style
    ) -> (ReturnValue:User Widget Class Reference)
   ) -> (then, ReturnValue:User Widget Object Reference)
   > n5 = Cast To WBP_UIKit_ButtonCTA_Base!(execute, Object=@n4) -> (then, CastFailed, AsWBP UIKit Button CTA Base:WBP UIKit Button CTA Base Object Reference) {is=title K2Node_DynamicCast}
   > SetHideInputAction!(execute, self:Common Button Base Object Reference=@n5, bInHideInputAction:bool=true) -> (then)
   > Set Text(Text=@n3, self:Fort CTAButton Object Reference=@n5)
   > Bind Event to On Clicked(self:Common Button Base Object Reference=@n5, Delegate=Create Event()) {is=title K2Node_AddDelegate}
   n6 = Get EntryBox_Buttons()
   n7 = "integer > integer"(A=GetNumEntries!(self:Dynamic Entry Box Base Object Reference=@n6) -> (ReturnValue:int=0))
   SetVisibility(
    execute=@n2.Completed
    self=@n6
    InVisibility=Select!(Option 0:ESlateVisibility Enum=Collapsed, Option 1:ESlateVisibility Enum=Visible, Index:bool=@n7) -> (ReturnValue:ESlateVisibility Enum)
   )
   > Branch(Condition=@n7)
   > Call On Success Finished Querying!(execute:Exec, self:WBP Community Post Content Object Reference) -> (then:Exec) {is=title K2Node_CallDelegate}
  ]}
 SetCompactUIStyle {nodes [
   SetCompactUIStyle() {is=title K2Node_FunctionEntry}
   > Branch(Condition=IsUsingCompactUI())
   > SetPadding!(
    execute
    self:Vertical Box Slot Object Reference=SlotAsVerticalBoxSlot(Widget=Get EntryBox_Buttons())
    InPadding_Left:float=0
    InPadding_Top:float=20
    InPadding_Right:float=0
    InPadding_Bottom:float=0
   ) -> (then)
   > Set Grid Slot Padding!(
    execute:Exec
    Widget:Widget Object Reference=Get HB_EventButtons() -> (HB_EventButtons:Horizontal Box Object Reference)
    Padding_Left:float=0
    Padding_Top:float=24
    Padding_Right:float=0
    Padding_Bottom:float=0
   ) -> (then:Exec) {is=title K2Node_MacroInstance}
   > Set Grid Slot Padding(Widget=Get SB_EventInfo() -> (SB_EventInfo:Size Box Object Reference), Padding_Left=0, Padding_Top=24, Padding_Right=0, Padding_Bottom=0)
  ]}
 SetContentQuerying {nodes [
   n1 = SetContentQuerying!() -> (then, StateType:ECommunityPostState Enum=Invalid) {is=title K2Node_FunctionEntry}
   > n2 = Switch on ECommunityPostState!(execute, Selection:ECommunityPostState Enum=@n1) -> (Invalid:Exec, Querying:Exec, Success:Exec, Failure:Exec) {is=title K2Node_SwitchEnum}
   SetVisibility(
    execute=@n2.Querying
    self=[
     Get GP_EventContent() -> (GP_EventContent:Grid Panel Object Reference)
     Get GP_Failed() -> (GP_Failed:Grid Panel Object Reference)
     Get GP_HighlightContent() -> (GP_HighlightContent:Grid Panel Object Reference)
    ]
    InVisibility=Collapsed
   )
   > SetVisibility(self=Get GP_Loading() -> (GP_Loading:Grid Panel Object Reference), InVisibility=SelfHitTestInvisible)
   SetVisibility(execute=[@n2.Failure, @n2.Invalid], self=[Get GP_Loading(), Get GP_EventContent(), Get GP_HighlightContent()], InVisibility=Collapsed)
   > SetVisibility(self=Get GP_Failed(), InVisibility=SelfHitTestInvisible)
   SetVisibility(execute=@n2.Success, self=[Get GP_Failed(), Get GP_HighlightContent(), Get GP_EventContent(), Get GP_Loading()], InVisibility=Collapsed)
   > SetVisibility(
    self=Select(
     Option 0:Widget Object Reference=Get GP_HighlightContent()
     Option 1:Widget Object Reference=Get GP_EventContent()
     Index=GetIsCreatorEvent()
    ) -> (ReturnValue:Widget Object Reference)
    InVisibility=SelfHitTestInvisible
   )
   > Call On Success Finished Querying()
  ]}
 SetEventImageURL {nodes [
   n1 = SetEventImageURL!() -> (then, URL:String) {is=title K2Node_FunctionEntry}
   > SetBrushFromURL!(
    execute
    self:Fort Lazy Image Object Reference=Get Image_Texture_Event() -> (Image_Texture_Event:Fort Lazy Image Object Reference)
    CMSImageURL:String=@n1
    bMatchSize:bool=false
    Delay:float=0
    DownloadPriority:ECMSDownloadPriority Enum=Normal
   ) -> (then)
  ]}
 SetEventThumbnailImageFallback {nodes [
   SetEventThumbnailImageFallback() {is=title K2Node_FunctionEntry}
   > n1 = GetDynamicMaterial(self=Get Image_Texture() -> (Image_Texture:Fort Lazy Image Object Reference))
   > SetTextureParameterValue(self=@n1, ParameterName=ImageTexture, Value=/FortniteCoreUI/UI/Community/Art/Textures/T_UI_CommunityPost_Banner_FallbackError)
  ]}
 SetEventUrlForQRCode {nodes [
   n1 = SetEventUrlForQRCode!() -> (then, URL:String) {is=title K2Node_FunctionEntry}
   > n2 = GenerateQRCodeTextureFromString!(execute, StringToEncode:String=@n1) -> (then, ReturnValue:Texture 2D Object Reference)
   > Set Filter!(execute, Filter:TextureFilter Enum=TF_Nearest, self:Texture Object Reference=@n2) -> (then, Output_Get:TextureFilter Enum=TF_Nearest)
   > SetBrushFromTexture!(
    execute
    self:Image Object Reference=Get Image_EventQRCode() -> (Image_EventQRCode:Image Object Reference)
    Texture:Texture 2D Object Reference=@n2
    bMatchSize:bool=false
   ) -> (then)
  ]}
 SetHighlightId {nodes [
   SetHighlightId!() -> (then, HighlightId:String) {is=title K2Node_FunctionEntry}
   > n1 = Construct Object from Class(Class=/Script/FortniteUI.CommunityHighlightVM) -> (ReturnValue:Community Highlight VM Object Reference)
   > SetCommunityPostVM!(execute, self, Viewmodel:Community Post VM Object Reference=@n1) -> (then)
   > n2 = Branch(Condition=GetIsCreatorEvent())
   n3 = Get HighlightId() -> (HighlightId:String)
   > InitializeAsEvent!(execute, self:Community Highlight VM Object Reference=@n1, EventId:String=@n3, bIncludeIslandDetails:bool=true) -> (then)
   Initialize!(execute=@n2.else, self:Community Highlight VM Object Reference=@n1, HighlightId:String=@n3, bIncludeIslandDetails:bool=true) -> (then)
  ]}
 SetImageURL {nodes [
   n1 = SetImageURL!() -> (then, URL:String) {is=title K2Node_FunctionEntry}
   > SetBrushFromURL(self=Get Image_Texture(), CMSImageURL=@n1, bMatchSize=false, Delay=0, DownloadPriority=Normal)
  ]}
 SetInterestedButtonVisibility {nodes [
   n1 = SetInterestedButtonVisibility!() -> (then, EventStatus:ECommunityEventStatus Enum=Unknown) {is=title K2Node_FunctionEntry}
   > SetInterestedButtonVisibility!(execute, self:WBP Community Interested Button Object Reference=Get Button_Interested(), EventStatus:ECommunityEventStatus Enum=@n1) -> (then) {is=title K2Node_CallFunction}
   n2 = Get EventStatus() -> (EventStatus:ECommunityEventStatus Enum=Unknown)
   > SetSize!(
    execute
    self:Horizontal Box Slot Object Reference=SlotAsHorizontalBoxSlot!(Widget:Widget Object Reference=Get Button_EventReplies() -> (Button_EventReplies:WBP UIKit Button Regular Object Reference)) -> (ReturnValue:Horizontal Box Slot Object Reference)
    InSize:Slate Child Size Structure=Make Slate Child Size!(
     Value:float=1
     SizeRule:ESlateSizeRule Enum=Select!(
      Unknown:ESlateSizeRule Enum=Fill
      Upcoming:ESlateSizeRule Enum=Fill
      Live:ESlateSizeRule Enum=Fill
      Ended:ESlateSizeRule Enum=Automatic
      Canceled:ESlateSizeRule Enum=Automatic
      Index:ECommunityEventStatus Enum=@n2
     ) -> (ReturnValue:ESlateSizeRule Enum)
    ) -> (SlateChildSize:Slate Child Size Structure) {is=title K2Node_MakeStruct}
   ) -> (then)
  ]}
 SetNavRules {nodes [
   SetNavRules() {is=title K2Node_FunctionEntry}
   n1 = Create Event()
   n2 = Get Button_Interested()
   > SetNavigationRuleCustom!(
    execute
    self:Widget Object Reference=Select!(Option 0:Widget Object Reference=Get Button_EventReaction(), Option 1:Widget Object Reference=@n2, Index:bool=IsVisible(self=@n2)) -> (ReturnValue:Widget Object Reference)
    Direction:EUINavigation Enum=Left
    InCustomDelegate:Delegate=@n1
   ) -> (then)
   > SetNavigationRuleCustom(self=Get Button_EventReplies(), Direction=Right, InCustomDelegate=@n1)
  ]}
 SetThumbnailImageFallback {nodes [
   SetThumbnailImageFallback() {is=title K2Node_FunctionEntry}
   > n1 = GetDynamicMaterial(self=Get Image_Texture())
   > SetTextureParameterValue(self=@n1, ParameterName=ImageTexture, Value=&T_UI_CommunityPost_Banner_FallbackError)
  ]}
 SetUrlForQRCode {nodes [
   n1 = SetUrlForQRCode!() -> (then, URL:String) {is=title K2Node_FunctionEntry}
   > n2 = GenerateQRCodeTextureFromString(StringToEncode=@n1)
   > Set Filter(Filter=TF_Nearest, self=@n2) -> (Output_Get=TF_Nearest)
   > SetBrushFromTexture(self=Get Image_QRCode() -> (Image_QRCode:Image Object Reference), Texture=@n2, bMatchSize=false)
  ]}
]
variables [
 IsEvent {type=boolean}
 IsReportable {type=boolean}
 IsUsingCompactUI {type=boolean}
 LastFocusedWidget {type=Widget, default=null}
 MotdButtons {type="array<MotdButton>", default []}
 OnCTAButtonClicked {}
 OnNavigate {}
 OnSuccessFinishedQuerying {}
]
widgets {
 info {inheritedWidgetCount=0, parentClass=&UserWidget, rootWidgetClass=&GridPanel, widgetCount=52}
 widgets [
  Root {
   widgetClassPath=&GridPanel
   widget=Root
   properties {
    clipping=ClipToBounds
    navigation=&self:WidgetTree.Root.WidgetNavigation_0
    columnFill [1]
    rowFill [1,0]
   }
  }
  CaptureForPostBufferUpdate {
   widgetClassPath=&WBP_CaptureForPostBufferUpdate
   parent=Root
   widget=CaptureForPostBufferUpdate
   slot {class=GridSlot, properties {rowSpan=2}}
  }
  GP_Loading {
   widgetClassPath=&GridPanel
   bIsVariable=true
   parent=Root
   widget=GP_Loading
   properties {columnFill [1], rowFill [1]}
   slot {class=GridSlot, properties {layer=2, rowSpan=2}}
  }
  WBP_UIKit_Throbber {
   widgetClassPath=/FortUILibrary/UIKit/Throbber/WBP_UIKit_Throbber
   bIsVariable=true
   parent=GP_Loading
   widget=WBP_UIKit_Throbber
   properties {throbberSize=Large}
   slot {class=GridSlot, properties {horizontalAlignment=HAlign_Center, verticalAlignment=VAlign_Center}}
  }
  GP_EventContent {
   widgetClassPath=&GridPanel
   bIsVariable=true
   parent=Root
   widget=GP_EventContent
   properties {columnFill [1.31,1], rowFill [1,0,0]}
   slot {class=GridSlot, properties {layer=2, rowSpan=2, padding=<ltrb 64 64 64 64>}}
  }
  SizeBox_0 {
   widgetClassPath=&SizeBox
   parent=GP_EventContent
   widget=SizeBox_0
   properties {bOverride_MaxAspectRatio=true, bOverride_MinAspectRatio=true, maxAspectRatio=1.7778, minAspectRatio=1.7778}
   slot {class=GridSlot, properties {columnSpan=1, horizontalAlignment=HAlign_Left, verticalAlignment=VAlign_Top}}
  }
  Image_Texture_Event {
   widgetClassPath=&FortLazyImage
   bIsVariable=true
   parent=SizeBox_0
   widget=Image_Texture_Event
   properties {
    materialTextureParamName=ImageTexture
    pixelSnapping=SnapToPixel
    brush {
     bIsDynamicallyLoaded=false
     drawAs=Image
     imageType=FullColor
     mirroring=NoMirror
     resourceName=None
     resourceObject=/FortniteCoreUI/UI/Community/Art/Materials/M_UI_Communities_Backgroud_Texture
     tiling=NoTile
     imageSize=<944 532>
     margin=<ltrb 0 0 0 0>
     outlineSettings {
      bUseBrushTransparency=false
      roundingType=HalfHeightRadius
      width=0
      color {colorUseRule=UseColor_Specified, specifiedColor=<rgba 0 0 0 0>}
      cornerRadii=<0 0 0 0>
     }
     tintColor {colorUseRule=UseColor_Specified, specifiedColor=<rgba 1 1 1 1>}
     uVRegion {bIsValid=false, max=<0 0>, min=<0 0>}
    }
    loadingBackgroundBrush {
     bIsDynamicallyLoaded=false
     drawAs=Image
     imageType=FullColor
     mirroring=NoMirror
     resourceName=None
     resourceObject=/FortniteCoreUI/UI/Community/Art/M_UI_CommunityPost_Loading
     tiling=NoTile
     imageSize=<864 486>
     margin=<ltrb 0 0 0 0>
     outlineSettings {
      bUseBrushTransparency=false
      roundingType=HalfHeightRadius
      width=0
      color {colorUseRule=UseColor_Specified, specifiedColor=<rgba 0 0 0 0>}
      cornerRadii=<0 0 0 0>
     }
     tintColor {colorUseRule=UseColor_Specified, specifiedColor=<rgba 1 1 1 1>}
     uVRegion {bIsValid=false, max=<0 0>, min=<0 0>}
    }
    loadingThrobberBrush {
     bIsDynamicallyLoaded=false
     drawAs=Image
     imageType=FullColor
     mirroring=NoMirror
     resourceName=None
     resourceObject=/Game/Global/Textures/Default/Blanks/T_Blank_White_ZeroAlpha_sRGBColor
     tiling=NoTile
     imageSize=<1 1>
     margin=<ltrb 0 0 0 0>
     outlineSettings {
      bUseBrushTransparency=false
      roundingType=HalfHeightRadius
      width=0
      color {colorUseRule=UseColor_Specified, specifiedColor=<rgba 0 0 0 0>}
      cornerRadii=<0 0 0 0>
     }
     tintColor {colorUseRule=UseColor_Specified, specifiedColor=<rgba 1 1 1 1>}
     uVRegion {bIsValid=false, max=<0 0>, min=<0 0>}
    }
   }
   slot {class=SizeBoxSlot}
  }
  SB_EventInfo {
   widgetClassPath=&SizeBox
   bIsVariable=true
   parent=GP_EventContent
   widget=SB_EventInfo
   properties {bOverride_HeightOverride=true, heightOverride=104}
   slot {class=GridSlot, properties {columnSpan=1, row=1, rowSpan=1, padding=<ltrb 0 73 0 0>}}
  }
  GP_EventInfo {
   widgetClassPath=&GridPanel
   parent=SB_EventInfo
   widget=GP_EventInfo
   properties {bOverride_Cursor=true, columnFill [0,1], rowFill [1]}
   slot {class=SizeBoxSlot}
  }
  GP_EventQRCode {
   widgetClassPath=&GridPanel
   bIsVariable=true
   parent=GP_EventInfo
   widget=GP_EventQRCode
   properties {bOverride_Cursor=true, visibility=Collapsed, columnFill [1], rowFill [1]}
   slot {class=GridSlot, properties {padding=<ltrb 0 0 64 0>}}
  }
  Image {
   widgetClassPath=&Image
   bIsVariable=true
   parent=GP_EventQRCode
   widget=Image
   properties {
    visibility=SelfHitTestInvisible
    brush {
     bIsDynamicallyLoaded=false
     drawAs=Image
     imageType=FullColor
     mirroring=NoMirror
     resourceName=None
     resourceObject=/FortniteCoreUI/UI/Community/Art/MaterialInstances/MI_UI_Community_QRBackplate
     tiling=NoTile
     imageSize=<104 104>
     margin=<ltrb 0 0 0 0>
     outlineSettings {
      bUseBrushTransparency=false
      roundingType=HalfHeightRadius
      width=0
      color {colorUseRule=UseColor_Specified, specifiedColor=<rgba 0 0 0 0>}
      cornerRadii=<0 0 0 0>
     }
     tintColor {colorUseRule=UseColor_Specified, specifiedColor=<rgba 1 1 1 1>}
     uVRegion {bIsValid=false, max=<0 0>, min=<0 0>}
    }
   }
   slot {class=GridSlot}
  }
  Image_EventQRCode {
   widgetClassPath=&Image
   bIsVariable=true
   parent=GP_EventQRCode
   widget=Image_EventQRCode
   properties {
    visibility=SelfHitTestInvisible
    brush {
     bIsDynamicallyLoaded=false
     drawAs=Image
     imageType=FullColor
     mirroring=NoMirror
     resourceName=None
     resourceObject=null
     tiling=NoTile
     imageSize=<90 90>
     margin=<ltrb 0 0 0 0>
     outlineSettings {
      bUseBrushTransparency=false
      roundingType=HalfHeightRadius
      width=0
      color {colorUseRule=UseColor_Specified, specifiedColor=<rgba 0 0 0 0>}
      cornerRadii=<0 0 0 0>
     }
     tintColor {colorUseRule=UseColor_Specified, specifiedColor=<rgba 1 1 1 1>}
     uVRegion {bIsValid=false, max=<0 0>, min=<0 0>}
    }
   }
   slot {class=GridSlot}
  }
  VerticalBox_184 {
   widgetClassPath=&VerticalBox
   parent=GP_EventQRCode
   widget=VerticalBox_184
   slot {class=GridSlot, properties {column=1, verticalAlignment=VAlign_Center, padding=<ltrb 24 0 0 0>}}
  }
  Text_ScanQR {
   widgetClassPath=&CommonTextBlock
   parent=VerticalBox_184
   widget=Text_ScanQR
   properties {
    bIsScrollingEnabled=false
    mobileFontSizeMultiplier=1.1
    text=Scan QR Code
    font {
     bForceMonospaced=false
     bMaterialIsStencil=false
     fontMaterial=null
     fontObject=&NotoSans
     letterSpacing=0
     monospacedWidth=1
     size=24
     skewAmount=0
     typefaceFontName=NotoSans-Regular
     outlineSettings {
      bApplyOutlineToDropShadows=false
      bMiteredCorners=false
      bSeparateFillAlpha=false
      outlineMaterial=null
      outlineSize=0
      outlineColor=<rgba 0 0 0 1>
     }
    }
    shadowOffset=<0 0>
   }
   slot {class=VerticalBoxSlot}
  }
  Text_MoreInfoQR {
   widgetClassPath=&CommonTextBlock
   parent=VerticalBox_184
   widget=Text_MoreInfoQR
   properties {
    bIsScrollingEnabled=false
    mobileFontSizeMultiplier=1.25
    text=For more information
    colorAndOpacity {colorUseRule=UseColor_Specified, specifiedColor=<rgba 0.7913 0.7913 0.8228 0.698>}
    font {
     bForceMonospaced=false
     bMaterialIsStencil=false
     fontMaterial=null
     fontObject=&HeadingNow
     letterSpacing=0
     monospacedWidth=1
     size=12.75
     skewAmount=0
     typefaceFontName=74 Regular
     outlineSettings {
      bApplyOutlineToDropShadows=false
      bMiteredCorners=false
      bSeparateFillAlpha=false
      outlineMaterial=null
      outlineSize=0
      outlineColor=<rgba 0 0 0 1>
     }
    }
    shadowOffset=<0 0>
   }
   slot {class=VerticalBoxSlot, properties {verticalAlignment=VAlign_Top}}
  }
  GP_EventCCU {
   widgetClassPath=&GridPanel
   bIsVariable=true
   parent=GP_EventInfo
   widget=GP_EventCCU
   properties {bOverride_Cursor=true, columnFill [1], rowFill [1,1]}
   slot {class=GridSlot, properties {column=1, horizontalAlignment=HAlign_Left}}
  }
  Image_322 {
   widgetClassPath=&Image
   parent=GP_EventCCU
   widget=Image_322
   properties {brush {
     bIsDynamicallyLoaded=false
     drawAs=Image
     imageType=FullColor
     mirroring=NoMirror
     resourceName=None
     resourceObject=/FortUILibrary/Textures/MaterialInstances/MI_UIKit_Icon_PartyUp
     tiling=NoTile
     imageSize=<64 64>
     margin=<ltrb 0 0 0 0>
     outlineSettings {
      bUseBrushTransparency=false
      roundingType=HalfHeightRadius
      width=0
      color {colorUseRule=UseColor_Specified, specifiedColor=<rgba 0 0 0 0>}
      cornerRadii=<0 0 0 0>
     }
     tintColor {colorUseRule=UseColor_Specified, specifiedColor=<rgba 1 1 1 1>}
     uVRegion {bIsValid=false, max=<0 0>, min=<0 0>}
    }}
   slot {class=GridSlot, properties {rowSpan=2, verticalAlignment=VAlign_Center, padding=<ltrb 0 0 0 3>}}
  }
  Text_EventInterestedCount {
   widgetClassPath=&CommonTextBlock
   bIsVariable=true
   parent=GP_EventCCU
   widget=Text_EventInterestedCount
   properties {
    bIsScrollingEnabled=false
    mobileFontSizeMultiplier=1.1
    text="{Count}"
    font {
     bForceMonospaced=false
     bMaterialIsStencil=false
     fontMaterial=null
     fontObject=&NotoSans
     letterSpacing=0
     monospacedWidth=1
     size=24
     skewAmount=0
     typefaceFontName=NotoSans-Regular
     outlineSettings {
      bApplyOutlineToDropShadows=false
      bMiteredCorners=false
      bSeparateFillAlpha=false
      outlineMaterial=null
      outlineSize=0
      outlineColor=<rgba 0 0 0 1>
     }
    }
    shadowOffset=<0 0>
   }
   slot {class=GridSlot, properties {column=1, verticalAlignment=VAlign_Bottom, padding=<ltrb 24 0 0 0>}}
  }
  Text_CCU_Interested {
   widgetClassPath=&CommonTextBlock
   parent=GP_EventCCU
   widget=Text_CCU_Interested
   properties {
    bIsScrollingEnabled=false
    mobileFontSizeMultiplier=1.25
    text=Interested
    colorAndOpacity {colorUseRule=UseColor_Specified, specifiedColor=<rgba 0.7913 0.7913 0.8228 0.698>}
    font {
     bForceMonospaced=false
     bMaterialIsStencil=false
     fontMaterial=null
     fontObject=&HeadingNow
     letterSpacing=0
     monospacedWidth=1
     size=12.75
     skewAmount=0
     typefaceFontName=74 Regular
     outlineSettings {
      bApplyOutlineToDropShadows=false
      bMiteredCorners=false
      bSeparateFillAlpha=false
      outlineMaterial=null
      outlineSize=0
      outlineColor=<rgba 0 0 0 1>
     }
    }
    shadowOffset=<0 0>
   }
   slot {class=GridSlot, properties {column=1, row=1, verticalAlignment=VAlign_Top, padding=<ltrb 24 0 0 0>}}
  }
  HB_EventButtons {
   widgetClassPath=&HorizontalBox
   bIsVariable=true
   parent=GP_EventContent
   widget=HB_EventButtons
   slot {class=GridSlot, properties {columnSpan=1, row=2, padding=<ltrb 0 73 0 0>}}
  }
  Button_Interested {
   widgetClassPath=&/WBP_Community_InterestedButton
   bIsVariable=true
   parent=HB_EventButtons
   widget=Button_Interested
   properties {
    bHideInputAction=true
    bHideInputActionWithKeyboard=true
    minWidth=325
    mobileMinWidth=325
    navigation=&self:WidgetTree.Button_Interested.WidgetNavigation_0
   }
   slot {class=HorizontalBoxSlot, properties {padding=<ltrb 0 0 24 0>}}
  }
  Button_EventReaction {
   widgetClassPath=&/WBP_CommunityPost_ReactionButton
   bIsVariable=true
   parent=HB_EventButtons
   widget=Button_EventReaction
   properties {bHideInputAction=true, bHideInputActionWithKeyboard=true, text="##"}
   slot {class=HorizontalBoxSlot, properties {padding=<ltrb 0 0 24 0>}}
  }
  Button_EventReplies {
   widgetClassPath=&WBP_UIKit_Button_Regular
   bIsVariable=true
   parent=HB_EventButtons
   widget=Button_EventReplies
   properties {
    bHideInputAction=true
    bHideInputActionWithKeyboard=true
    navigation=&self:WidgetTree.Button_EventReplies.WidgetNavigation_0
    text="View ## Replies"
   }
   slot {class=HorizontalBoxSlot, properties {size {sizeRule=Fill, value=1}}}
  }
  SB_ContentWidthLimiterEvent {
   widgetClassPath=&SizeBox
   parent=GP_EventContent
   widget=SB_ContentWidthLimiterEvent
   properties {bOverride_HeightOverride=true, bOverride_WidthOverride=true, widthOverride=600}
   slot {class=GridSlot, properties {column=1, horizontalAlignment=HAlign_Right, rowSpan=3}}
  }
  VerticalBox_182 {widgetClassPath=&VerticalBox, parent=SB_ContentWidthLimiterEvent, widget=VerticalBox_182, slot {class=SizeBoxSlot}}
  CommunityPost_CreatorLine {
   widgetClassPath=&/WBP_CommunityPost_CreatorLine
   bIsVariable=true
   parent=VerticalBox_182
   widget=CommunityPost_CreatorLine
   slot {class=VerticalBoxSlot, properties {padding=<ltrb 0 0 0 32>}}
  }
  ScrollBox_Event {
   widgetClassPath=&ScrollBox
   bIsVariable=true
   parent=VerticalBox_182
   widget=ScrollBox_Event
   properties {
    bAnimateWheelScrolling=true
    bConsumePointerInput=false
    scrollbarPadding=<ltrb 0 0 -10 0>
    scrollbarThickness=<8 8>
    widgetBarStyle {
     thickness=12
     draggedThumbImage {
      bIsDynamicallyLoaded=false
      drawAs=RoundedBox
      imageType=FullColor
      mirroring=NoMirror
      resourceName=../../../Engine/Content/Slate/Common/Scrollbar_Thumb.png
      resourceObject=null
      tiling=NoTile
      imageSize=<12 12>
      margin=<ltrb 0.25 0.25 0.25 0.25>
      outlineSettings {
       bUseBrushTransparency=false
       roundingType=FixedRadius
       width=2
       color {colorUseRule=UseColor_Specified, specifiedColor=<rgba 0 0 0 0>}
       cornerRadii=<6 6 6 6>
      }
      tintColor {colorUseRule=UseColor_Specified, specifiedColor=<rgba 1 1 1 1>}
      uVRegion {bIsValid=false, max=<0 0>, min=<0 0>}
     }
     horizontalBackgroundImage {
      bIsDynamicallyLoaded=false
      drawAs=NoDrawType
      imageType=NoImage
      mirroring=NoMirror
      resourceName=None
      resourceObject=null
      tiling=NoTile
      imageSize=<0 0>
      margin=<ltrb 0 0 0 0>
      outlineSettings {
       bUseBrushTransparency=false
       roundingType=HalfHeightRadius
       width=0
       color {colorUseRule=UseColor_Specified, specifiedColor=<rgba 0 0 0 0>}
       cornerRadii=<0 0 0 0>
      }
      tintColor {colorUseRule=UseColor_Specified, specifiedColor=<rgba 1 1 1 1>}
      uVRegion {bIsValid=false, max=<0 0>, min=<0 0>}
     }
     horizontalBottomSlotImage {
      bIsDynamicallyLoaded=false
      drawAs=NoDrawType
      imageType=FullColor
      mirroring=NoMirror
      resourceName=../../../Engine/Content/Slate/Common/Scrollbar_Background_Horizontal.png
      resourceObject=null
      tiling=NoTile
      imageSize=<8 8>
      margin=<ltrb 0 0 0 0>
      outlineSettings {
       bUseBrushTransparency=false
       roundingType=HalfHeightRadius
       width=0
       color {colorUseRule=UseColor_Specified, specifiedColor=<rgba 0 0 0 0>}
       cornerRadii=<0 0 0 0>
      }
      tintColor {colorUseRule=UseColor_Specified, specifiedColor=<rgba 1 1 1 1>}
      uVRegion {bIsValid=false, max=<0 0>, min=<0 0>}
     }
     horizontalTopSlotImage {
      bIsDynamicallyLoaded=false
      drawAs=NoDrawType
      imageType=FullColor
      mirroring=NoMirror
      resourceName=../../../Engine/Content/Slate/Common/Scrollbar_Background_Horizontal.png
      resourceObject=null
      tiling=NoTile
      imageSize=<8 8>
      margin=<ltrb 0 0 0 0>
      outlineSettings {
       bUseBrushTransparency=false
       roundingType=HalfHeightRadius
       width=0
       color {colorUseRule=UseColor_Specified, specifiedColor=<rgba 0 0 0 0>}
       cornerRadii=<0 0 0 0>
      }
      tintColor {colorUseRule=UseColor_Specified, specifiedColor=<rgba 1 1 1 1>}
      uVRegion {bIsValid=false, max=<0 0>, min=<0 0>}
     }
     hoveredThumbImage {
      bIsDynamicallyLoaded=false
      drawAs=RoundedBox
      imageType=FullColor
      mirroring=NoMirror
      resourceName=../../../Engine/Content/Slate/Common/Scrollbar_Thumb.png
      resourceObject=null
      tiling=NoTile
      imageSize=<12 12>
      margin=<ltrb 0.25 0.25 0.25 0.25>
      outlineSettings {
       bUseBrushTransparency=false
       roundingType=FixedRadius
       width=2
       color {colorUseRule=UseColor_Specified, specifiedColor=<rgba 0 0 0 0>}
       cornerRadii=<6 6 6 6>
      }
      tintColor {colorUseRule=UseColor_Specified, specifiedColor=<rgba 1 1 1 1>}
      uVRegion {bIsValid=false, max=<0 0>, min=<0 0>}
     }
     normalThumbImage {
      bIsDynamicallyLoaded=false
      drawAs=RoundedBox
      imageType=FullColor
      mirroring=NoMirror
      resourceName=../../../Engine/Content/Slate/Common/Scrollbar_Thumb.png
      resourceObject=null
      tiling=NoTile
      imageSize=<12 12>
      margin=<ltrb 0.25 0.25 0.25 0.25>
      outlineSettings {
       bUseBrushTransparency=false
       roundingType=FixedRadius
       width=2
       color {colorUseRule=UseColor_Specified, specifiedColor=<rgba 0 0 0 0>}
       cornerRadii=<6 6 6 6>
      }
      tintColor {colorUseRule=UseColor_Specified, specifiedColor=<rgba 1 1 1 0.749>}
      uVRegion {bIsValid=false, max=<0 0>, min=<0 0>}
     }
     verticalBackgroundImage {
      bIsDynamicallyLoaded=false
      drawAs=Box
      imageType=FullColor
      mirroring=NoMirror
      resourceName=None
      resourceObject=/FortUILibrary/UIKit/ScrollBar/MaterialInstances/MI_UIKit_ScrollbarBackground
      tiling=NoTile
      imageSize=<12 12>
      margin=<ltrb 0.5 0.5 0.5 0.5>
      outlineSettings {
       bUseBrushTransparency=false
       roundingType=HalfHeightRadius
       width=0
       color {colorUseRule=UseColor_Specified, specifiedColor=<rgba 0 0 0 0>}
       cornerRadii=<0 0 0 0>
      }
      tintColor {colorUseRule=UseColor_Specified, specifiedColor=<rgba 0.521 0.521 0.521 0.502>}
      uVRegion {bIsValid=false, max=<0 0>, min=<0 0>}
     }
     verticalBottomSlotImage {
      bIsDynamicallyLoaded=false
      drawAs=NoDrawType
      imageType=FullColor
      mirroring=NoMirror
      resourceName=../../../Engine/Content/Slate/Common/Scrollbar_Background_Vertical.png
      resourceObject=null
      tiling=NoTile
      imageSize=<8 8>
      margin=<ltrb 0 0 0 0>
      outlineSettings {
       bUseBrushTransparency=false
       roundingType=HalfHeightRadius
       width=0
       color {colorUseRule=UseColor_Specified, specifiedColor=<rgba 0 0 0 0>}
       cornerRadii=<0 0 0 0>
      }
      tintColor {colorUseRule=UseColor_Specified, specifiedColor=<rgba 1 1 1 1>}
      uVRegion {bIsValid=false, max=<0 0>, min=<0 0>}
     }
     verticalTopSlotImage {
      bIsDynamicallyLoaded=false
      drawAs=NoDrawType
      imageType=FullColor
      mirroring=NoMirror
      resourceName=../../../Engine/Content/Slate/Common/Scrollbar_Background_Vertical.png
      resourceObject=null
      tiling=NoTile
      imageSize=<8 8>
      margin=<ltrb 0 0 0 0>
      outlineSettings {
       bUseBrushTransparency=false
       roundingType=HalfHeightRadius
       width=0
       color {colorUseRule=UseColor_Specified, specifiedColor=<rgba 0 0 0 0>}
       cornerRadii=<0 0 0 0>
      }
      tintColor {colorUseRule=UseColor_Specified, specifiedColor=<rgba 1 1 1 1>}
      uVRegion {bIsValid=false, max=<0 0>, min=<0 0>}
     }
    }
    widgetStyle {
     barThickness=12
     bottomShadowBrush {
      bIsDynamicallyLoaded=false
      drawAs=NoDrawType
      imageType=FullColor
      mirroring=NoMirror
      resourceName=../../../Engine/Content/Slate/Common/ScrollBoxShadowBottom.png
      resourceObject=null
      tiling=NoTile
      imageSize=<16 8>
      margin=<ltrb 0.5 0 0.5 1>
      outlineSettings {
       bUseBrushTransparency=false
       roundingType=HalfHeightRadius
       width=0
       color {colorUseRule=UseColor_Specified, specifiedColor=<rgba 0 0 0 0>}
       cornerRadii=<0 0 0 0>
      }
      tintColor {colorUseRule=UseColor_Specified, specifiedColor=<rgba 1 1 1 1>}
      uVRegion {bIsValid=false, max=<0 0>, min=<0 0>}
     }
     horizontalScrolledContentPadding=<ltrb 0 0 1 0>
     leftShadowBrush {
      bIsDynamicallyLoaded=false
      drawAs=NoDrawType
      imageType=FullColor
      mirroring=NoMirror
      resourceName=../../../Engine/Content/Slate/Common/ScrollBoxShadowLeft.png
      resourceObject=null
      tiling=NoTile
      imageSize=<8 16>
      margin=<ltrb 1 0.5 0 0.5>
      outlineSettings {
       bUseBrushTransparency=false
       roundingType=HalfHeightRadius
       width=0
       color {colorUseRule=UseColor_Specified, specifiedColor=<rgba 0 0 0 0>}
       cornerRadii=<0 0 0 0>
      }
      tintColor {colorUseRule=UseColor_Specified, specifiedColor=<rgba 1 1 1 1>}
      uVRegion {bIsValid=false, max=<0 0>, min=<0 0>}
     }
     rightShadowBrush {
      bIsDynamicallyLoaded=false
      drawAs=NoDrawType
      imageType=FullColor
      mirroring=NoMirror
      resourceName=../../../Engine/Content/Slate/Common/ScrollBoxShadowRight.png
      resourceObject=null
      tiling=NoTile
      imageSize=<8 16>
      margin=<ltrb 0 0.5 1 0.5>
      outlineSettings {
       bUseBrushTransparency=false
       roundingType=HalfHeightRadius
       width=0
       color {colorUseRule=UseColor_Specified, specifiedColor=<rgba 0 0 0 0>}
       cornerRadii=<0 0 0 0>
      }
      tintColor {colorUseRule=UseColor_Specified, specifiedColor=<rgba 1 1 1 1>}
      uVRegion {bIsValid=false, max=<0 0>, min=<0 0>}
     }
     topShadowBrush {
      bIsDynamicallyLoaded=false
      drawAs=NoDrawType
      imageType=FullColor
      mirroring=NoMirror
      resourceName=../../../Engine/Content/Slate/Common/ScrollBoxShadowTop.png
      resourceObject=null
      tiling=NoTile
      imageSize=<16 8>
      margin=<ltrb 0.5 1 0.5 0>
      outlineSettings {
       bUseBrushTransparency=false
       roundingType=HalfHeightRadius
       width=0
       color {colorUseRule=UseColor_Specified, specifiedColor=<rgba 0 0 0 0>}
       cornerRadii=<0 0 0 0>
      }
      tintColor {colorUseRule=UseColor_Specified, specifiedColor=<rgba 1 1 1 1>}
      uVRegion {bIsValid=false, max=<0 0>, min=<0 0>}
     }
     verticalScrolledContentPadding=<ltrb 0 0 0 1>
    }
   }
   slot {class=VerticalBoxSlot, properties {size {sizeRule=Fill, value=1}}}
  }
  VerticalBox_242 {
   widgetClassPath=&VerticalBox
   parent=ScrollBox_Event
   widget=VerticalBox_242
   slot {class=ScrollBoxSlot, properties {size {sizeRule=Fill, value=1}}}
  }
  SizeBox_1 {
   widgetClassPath=&SizeBox
   parent=VerticalBox_242
   widget=SizeBox_1
   properties {bOverride_MaxDesiredHeight=true, clipping=ClipToBounds, maxDesiredHeight=180}
   slot {class=VerticalBoxSlot}
  }
  Text_EventTitle {
   widgetClassPath=&CommonTextBlock
   bIsVariable=true
   parent=SizeBox_1
   widget=Text_EventTitle
   properties {
    autoWrapText=true
    text="Lorem ipsum dolor sit amet, consectetur adipiscing elit"
    textOverflowPolicy=MultilineEllipsis
    font {
     bForceMonospaced=false
     bMaterialIsStencil=false
     fontMaterial=null
     fontObject=&HeadingNow
     letterSpacing=0
     monospacedWidth=1
     size=24
     skewAmount=0
     typefaceFontName=86 Bold
     outlineSettings {
      bApplyOutlineToDropShadows=false
      bMiteredCorners=false
      bSeparateFillAlpha=false
      outlineMaterial=null
      outlineSize=0
      outlineColor=<rgba 0 0 0 1>
     }
    }
    shadowOffset=<0 0>
   }
   slot {class=SizeBoxSlot}
  }
  WBP_CommunityPost_DateInfo {
   widgetClassPath=&/WBP_CommunityPost_DateInfo
   bIsVariable=true
   parent=VerticalBox_242
   widget=WBP_CommunityPost_DateInfo
   slot {class=VerticalBoxSlot, properties {padding=<ltrb 0 32 0 32>}}
  }
  RichTextEvent_Body {
   widgetClassPath=&CommonRichTextBlock
   bIsVariable=true
   parent=VerticalBox_242
   widget=RichTextEvent_Body
   properties {
    autoWrapText=true
    bOverrideDefaultStyle=true
    lineHeightPercentage=1.15
    mobileTextBlockScale=1.15
    text="Lorem ipsum dolor sit amet, consectetur adipiscing elit. Suspendisse quis neque a neque efficitur pellentesque. Nam vitae rhoncus ipsum. Donec viverra quis justo quis pellentesque. Nunc facilisis lectus eget augue condimentum porttitor. Sed finibus magna eu metus ornare dignissim. Morbi risus est, hendrerit eu leo fringilla, bibendum ullamcorper erat. Pellentesque accumsan massa vitae consequat malesuada.\r\n\r\nVestibulum ante ipsum primis in faucibus orci luctus et ultrices posuere cubilia curae; Aenean nisi nunc, consequat non accumsan id, efficitur ut sem. Nullam at erat fringilla, efficitur eros tincidunt, pellentesque ante. Pellentesque dignissim lobortis malesuada. Fusce non aliquam sapien, nec accumsan ipsum. Quisque et porttitor tellus. In sollicitudin ligula nisl, non volutpat sem ultrices in. Proin maximus velit et ex cursus venenatis."
    textStyleSet=/FortniteCoreUI/UI/Community/Art/DT_Community_RichTextStyles
    defaultTextStyleOverride {
     overflowPolicy=Clip
     transformPolicy=None
     colorAndOpacity {colorUseRule=UseColor_Specified, specifiedColor=<rgba 0.5776 0.5776 0.5776 1>}
     font {
      bForceMonospaced=false
      bMaterialIsStencil=false
      fontMaterial=null
      fontObject=&HeadingNow
      letterSpacing=10
      monospacedWidth=1
      size=15.75
      skewAmount=0
      typefaceFontName=74 Regular
      outlineSettings {
       bApplyOutlineToDropShadows=false
       bMiteredCorners=false
       bSeparateFillAlpha=false
       outlineMaterial=null
       outlineSize=0
       outlineColor=<rgba 0 0 0 1>
      }
     }
     highlightColor {colorUseRule=UseColor_Specified, specifiedColor=<rgba 0 0 0 0>}
     highlightShape {
      bIsDynamicallyLoaded=false
      drawAs=Image
      imageType=NoImage
      mirroring=NoMirror
      resourceName=None
      resourceObject=null
      tiling=NoTile
      imageSize=<32 32>
      margin=<ltrb 0 0 0 0>
      outlineSettings {
       bUseBrushTransparency=false
       roundingType=HalfHeightRadius
       width=0
       color {colorUseRule=UseColor_Specified, specifiedColor=<rgba 0 0 0 0>}
       cornerRadii=<0 0 0 1>
      }
      tintColor {colorUseRule=UseColor_Specified, specifiedColor=<rgba 1 1 1 1>}
      uVRegion {bIsValid=false, max=<0 0>, min=<0 0>}
     }
     selectedBackgroundColor {colorUseRule=UseColor_Foreground, specifiedColor=<rgba 1 0 1 1>}
     shadowColorAndOpacity=<rgba 0 0 0 1>
     shadowOffset=<0 0>
     strikeBrush {
      bIsDynamicallyLoaded=false
      drawAs=Image
      imageType=NoImage
      mirroring=NoMirror
      resourceName=None
      resourceObject=null
      tiling=NoTile
      imageSize=<32 32>
      margin=<ltrb 0 0 0 0>
      outlineSettings {
       bUseBrushTransparency=false
       roundingType=HalfHeightRadius
       width=0
       color {colorUseRule=UseColor_Specified, specifiedColor=<rgba 0 0 0 0>}
       cornerRadii=<0 0 0 1>
      }
      tintColor {colorUseRule=UseColor_Specified, specifiedColor=<rgba 1 1 1 1>}
      uVRegion {bIsValid=false, max=<0 0>, min=<0 0>}
     }
     underlineBrush {
      bIsDynamicallyLoaded=false
      drawAs=Image
      imageType=NoImage
      mirroring=NoMirror
      resourceName=None
      resourceObject=null
      tiling=NoTile
      imageSize=<32 32>
      margin=<ltrb 0 0 0 0>
      outlineSettings {
       bUseBrushTransparency=false
       roundingType=HalfHeightRadius
       width=0
       color {colorUseRule=UseColor_Specified, specifiedColor=<rgba 0 0 0 0>}
       cornerRadii=<0 0 0 1>
      }
      tintColor {colorUseRule=UseColor_Specified, specifiedColor=<rgba 1 1 1 1>}
      uVRegion {bIsValid=false, max=<0 0>, min=<0 0>}
     }
    }
   }
   slot {class=VerticalBoxSlot}
  }
  GP_HighlightContent {
   widgetClassPath=&GridPanel
   bIsVariable=true
   parent=Root
   widget=GP_HighlightContent
   properties {visibility=Collapsed, columnFill [1,0], rowFill [1]}
   slot {class=GridSlot, properties {layer=2, rowSpan=2}}
  }
  Image_Texture {
   widgetClassPath=&FortLazyImage
   bIsVariable=true
   parent=GP_HighlightContent
   widget=Image_Texture
   properties {
    materialTextureParamName=ImageTexture
    pixelSnapping=SnapToPixel
    brush {
     bIsDynamicallyLoaded=false
     drawAs=Image
     imageType=FullColor
     mirroring=NoMirror
     resourceName=None
     resourceObject=&M_UI_Communities_Backgroud_Texture
     tiling=NoTile
     imageSize=<944 532>
     margin=<ltrb 0 0 0 0>
     outlineSettings {
      bUseBrushTransparency=false
      roundingType=HalfHeightRadius
      width=0
      color {colorUseRule=UseColor_Specified, specifiedColor=<rgba 0 0 0 0>}
      cornerRadii=<0 0 0 0>
     }
     tintColor {colorUseRule=UseColor_Specified, specifiedColor=<rgba 1 1 1 1>}
     uVRegion {bIsValid=false, max=<0 0>, min=<0 0>}
    }
    loadingBackgroundBrush {
     bIsDynamicallyLoaded=false
     drawAs=Image
     imageType=FullColor
     mirroring=NoMirror
     resourceName=None
     resourceObject=&M_UI_CommunityPost_Loading
     tiling=NoTile
     imageSize=<1160 1080>
     margin=<ltrb 0 0 0 0>
     outlineSettings {
      bUseBrushTransparency=false
      roundingType=HalfHeightRadius
      width=0
      color {colorUseRule=UseColor_Specified, specifiedColor=<rgba 0 0 0 0>}
      cornerRadii=<0 0 0 0>
     }
     tintColor {colorUseRule=UseColor_Specified, specifiedColor=<rgba 1 1 1 1>}
     uVRegion {bIsValid=false, max=<0 0>, min=<0 0>}
    }
    loadingThrobberBrush {
     bIsDynamicallyLoaded=false
     drawAs=Image
     imageType=FullColor
     mirroring=NoMirror
     resourceName=None
     resourceObject=&T_Blank_White_ZeroAlpha_sRGBColor
     tiling=NoTile
     imageSize=<1 1>
     margin=<ltrb 0 0 0 0>
     outlineSettings {
      bUseBrushTransparency=false
      roundingType=HalfHeightRadius
      width=0
      color {colorUseRule=UseColor_Specified, specifiedColor=<rgba 0 0 0 0>}
      cornerRadii=<0 0 0 0>
     }
     tintColor {colorUseRule=UseColor_Specified, specifiedColor=<rgba 1 1 1 1>}
     uVRegion {bIsValid=false, max=<0 0>, min=<0 0>}
    }
   }
   slot {class=GridSlot, properties {column=1, horizontalAlignment=HAlign_Right, rowSpan=2, verticalAlignment=VAlign_Top, padding=<ltrb 2 66 54 2>}}
  }
  O_Content {
   widgetClassPath=&Overlay
   parent=GP_HighlightContent
   widget=O_Content
   properties {bOverride_Cursor=true, flowDirectionPreference=Culture}
   slot {class=GridSlot, properties {horizontalAlignment=HAlign_Left, layer=2, rowSpan=2, padding=<ltrb 62 60 63 56>}}
  }
  SB_ContentWidthLimiter {
   widgetClassPath=&SizeBox
   bIsVariable=true
   parent=O_Content
   widget=SB_ContentWidthLimiter
   properties {bOverride_WidthOverride=true, widthOverride=552}
   slot {class=OverlaySlot, properties {horizontalAlignment=HAlign_Right, verticalAlignment=VAlign_Fill}}
  }
  VB {
   widgetClassPath=&VerticalBox
   parent=SB_ContentWidthLimiter
   widget=VB
   properties {bOverride_Cursor=true}
   slot {class=SizeBoxSlot}
  }
  CreatorLine {
   widgetClassPath=&WBP_CommunityPost_CreatorLine
   bIsVariable=true
   parent=VB
   widget=CreatorLine
   slot {class=VerticalBoxSlot, properties {padding=<ltrb 0 0 0 32>}}
  }
  Text_Title {
   widgetClassPath=&CommonTextBlock
   parent=VB
   widget=Text_Title
   properties {
    autoWrapText=true
    text=Title Text
    font {
     bForceMonospaced=false
     bMaterialIsStencil=false
     fontMaterial=null
     fontObject=&HeadingNow
     letterSpacing=10
     monospacedWidth=1
     size=25.5
     skewAmount=0
     typefaceFontName=86 Bold Italic
     outlineSettings {
      bApplyOutlineToDropShadows=false
      bMiteredCorners=false
      bSeparateFillAlpha=false
      outlineMaterial=null
      outlineSize=0
      outlineColor=<rgba 0 0 0 1>
     }
    }
    shadowOffset=<0 0>
   }
   slot {class=VerticalBoxSlot, properties {padding=<ltrb 0 0 0 12>}}
  }
  ScrollBox {
   widgetClassPath=&ScrollBox
   bIsVariable=true
   parent=VB
   widget=ScrollBox
   properties {
    bAnimateWheelScrolling=true
    bConsumePointerInput=false
    scrollbarPadding=<ltrb 8 0 0 0>
    scrollbarThickness=<12 12>
    widgetBarStyle {
     thickness=12
     draggedThumbImage {
      bIsDynamicallyLoaded=false
      drawAs=RoundedBox
      imageType=FullColor
      mirroring=NoMirror
      resourceName=../../../Engine/Content/Slate/Common/Scrollbar_Thumb.png
      resourceObject=null
      tiling=NoTile
      imageSize=<12 12>
      margin=<ltrb 0.25 0.25 0.25 0.25>
      outlineSettings {
       bUseBrushTransparency=false
       roundingType=FixedRadius
       width=2
       color {colorUseRule=UseColor_Specified, specifiedColor=<rgba 0 0 0 0>}
       cornerRadii=<6 6 6 6>
      }
      tintColor {colorUseRule=UseColor_Specified, specifiedColor=<rgba 1 1 1 1>}
      uVRegion {bIsValid=false, max=<0 0>, min=<0 0>}
     }
     horizontalBackgroundImage {
      bIsDynamicallyLoaded=false
      drawAs=NoDrawType
      imageType=NoImage
      mirroring=NoMirror
      resourceName=None
      resourceObject=null
      tiling=NoTile
      imageSize=<0 0>
      margin=<ltrb 0 0 0 0>
      outlineSettings {
       bUseBrushTransparency=false
       roundingType=HalfHeightRadius
       width=0
       color {colorUseRule=UseColor_Specified, specifiedColor=<rgba 0 0 0 0>}
       cornerRadii=<0 0 0 0>
      }
      tintColor {colorUseRule=UseColor_Specified, specifiedColor=<rgba 1 1 1 1>}
      uVRegion {bIsValid=false, max=<0 0>, min=<0 0>}
     }
     horizontalBottomSlotImage {
      bIsDynamicallyLoaded=false
      drawAs=NoDrawType
      imageType=FullColor
      mirroring=NoMirror
      resourceName=../../../Engine/Content/Slate/Common/Scrollbar_Background_Horizontal.png
      resourceObject=null
      tiling=NoTile
      imageSize=<8 8>
      margin=<ltrb 0 0 0 0>
      outlineSettings {
       bUseBrushTransparency=false
       roundingType=HalfHeightRadius
       width=0
       color {colorUseRule=UseColor_Specified, specifiedColor=<rgba 0 0 0 0>}
       cornerRadii=<0 0 0 0>
      }
      tintColor {colorUseRule=UseColor_Specified, specifiedColor=<rgba 1 1 1 1>}
      uVRegion {bIsValid=false, max=<0 0>, min=<0 0>}
     }
     horizontalTopSlotImage {
      bIsDynamicallyLoaded=false
      drawAs=NoDrawType
      imageType=FullColor
      mirroring=NoMirror
      resourceName=../../../Engine/Content/Slate/Common/Scrollbar_Background_Horizontal.png
      resourceObject=null
      tiling=NoTile
      imageSize=<8 8>
      margin=<ltrb 0 0 0 0>
      outlineSettings {
       bUseBrushTransparency=false
       roundingType=HalfHeightRadius
       width=0
       color {colorUseRule=UseColor_Specified, specifiedColor=<rgba 0 0 0 0>}
       cornerRadii=<0 0 0 0>
      }
      tintColor {colorUseRule=UseColor_Specified, specifiedColor=<rgba 1 1 1 1>}
      uVRegion {bIsValid=false, max=<0 0>, min=<0 0>}
     }
     hoveredThumbImage {
      bIsDynamicallyLoaded=false
      drawAs=RoundedBox
      imageType=FullColor
      mirroring=NoMirror
      resourceName=../../../Engine/Content/Slate/Common/Scrollbar_Thumb.png
      resourceObject=null
      tiling=NoTile
      imageSize=<12 12>
      margin=<ltrb 0.25 0.25 0.25 0.25>
      outlineSettings {
       bUseBrushTransparency=false
       roundingType=FixedRadius
       width=2
       color {colorUseRule=UseColor_Specified, specifiedColor=<rgba 0 0 0 0>}
       cornerRadii=<6 6 6 6>
      }
      tintColor {colorUseRule=UseColor_Specified, specifiedColor=<rgba 1 1 1 1>}
      uVRegion {bIsValid=false, max=<0 0>, min=<0 0>}
     }
     normalThumbImage {
      bIsDynamicallyLoaded=false
      drawAs=RoundedBox
      imageType=FullColor
      mirroring=NoMirror
      resourceName=../../../Engine/Content/Slate/Common/Scrollbar_Thumb.png
      resourceObject=null
      tiling=NoTile
      imageSize=<12 12>
      margin=<ltrb 0.25 0.25 0.25 0.25>
      outlineSettings {
       bUseBrushTransparency=false
       roundingType=FixedRadius
       width=2
       color {colorUseRule=UseColor_Specified, specifiedColor=<rgba 0 0 0 0>}
       cornerRadii=<6 6 6 6>
      }
      tintColor {colorUseRule=UseColor_Specified, specifiedColor=<rgba 1 1 1 0.749>}
      uVRegion {bIsValid=false, max=<0 0>, min=<0 0>}
     }
     verticalBackgroundImage {
      bIsDynamicallyLoaded=false
      drawAs=Box
      imageType=FullColor
      mirroring=NoMirror
      resourceName=None
      resourceObject=&MI_UIKit_ScrollbarBackground
      tiling=NoTile
      imageSize=<12 12>
      margin=<ltrb 0.5 0.5 0.5 0.5>
      outlineSettings {
       bUseBrushTransparency=false
       roundingType=HalfHeightRadius
       width=0
       color {colorUseRule=UseColor_Specified, specifiedColor=<rgba 0 0 0 0>}
       cornerRadii=<0 0 0 0>
      }
      tintColor {colorUseRule=UseColor_Specified, specifiedColor=<rgba 0.521 0.521 0.521 0.502>}
      uVRegion {bIsValid=false, max=<0 0>, min=<0 0>}
     }
     verticalBottomSlotImage {
      bIsDynamicallyLoaded=false
      drawAs=NoDrawType
      imageType=FullColor
      mirroring=NoMirror
      resourceName=../../../Engine/Content/Slate/Common/Scrollbar_Background_Vertical.png
      resourceObject=null
      tiling=NoTile
      imageSize=<8 8>
      margin=<ltrb 0 0 0 0>
      outlineSettings {
       bUseBrushTransparency=false
       roundingType=HalfHeightRadius
       width=0
       color {colorUseRule=UseColor_Specified, specifiedColor=<rgba 0 0 0 0>}
       cornerRadii=<0 0 0 0>
      }
      tintColor {colorUseRule=UseColor_Specified, specifiedColor=<rgba 1 1 1 1>}
      uVRegion {bIsValid=false, max=<0 0>, min=<0 0>}
     }
     verticalTopSlotImage {
      bIsDynamicallyLoaded=false
      drawAs=NoDrawType
      imageType=FullColor
      mirroring=NoMirror
      resourceName=../../../Engine/Content/Slate/Common/Scrollbar_Background_Vertical.png
      resourceObject=null
      tiling=NoTile
      imageSize=<8 8>
      margin=<ltrb 0 0 0 0>
      outlineSettings {
       bUseBrushTransparency=false
       roundingType=HalfHeightRadius
       width=0
       color {colorUseRule=UseColor_Specified, specifiedColor=<rgba 0 0 0 0>}
       cornerRadii=<0 0 0 0>
      }
      tintColor {colorUseRule=UseColor_Specified, specifiedColor=<rgba 1 1 1 1>}
      uVRegion {bIsValid=false, max=<0 0>, min=<0 0>}
     }
    }
    widgetStyle {
     barThickness=12
     bottomShadowBrush {
      bIsDynamicallyLoaded=false
      drawAs=NoDrawType
      imageType=FullColor
      mirroring=NoMirror
      resourceName=../../../Engine/Content/Slate/Common/ScrollBoxShadowBottom.png
      resourceObject=null
      tiling=NoTile
      imageSize=<16 8>
      margin=<ltrb 0.5 0 0.5 1>
      outlineSettings {
       bUseBrushTransparency=false
       roundingType=HalfHeightRadius
       width=0
       color {colorUseRule=UseColor_Specified, specifiedColor=<rgba 0 0 0 0>}
       cornerRadii=<0 0 0 0>
      }
      tintColor {colorUseRule=UseColor_Specified, specifiedColor=<rgba 1 1 1 1>}
      uVRegion {bIsValid=false, max=<0 0>, min=<0 0>}
     }
     horizontalScrolledContentPadding=<ltrb 0 0 1 0>
     leftShadowBrush {
      bIsDynamicallyLoaded=false
      drawAs=NoDrawType
      imageType=FullColor
      mirroring=NoMirror
      resourceName=../../../Engine/Content/Slate/Common/ScrollBoxShadowLeft.png
      resourceObject=null
      tiling=NoTile
      imageSize=<8 16>
      margin=<ltrb 1 0.5 0 0.5>
      outlineSettings {
       bUseBrushTransparency=false
       roundingType=HalfHeightRadius
       width=0
       color {colorUseRule=UseColor_Specified, specifiedColor=<rgba 0 0 0 0>}
       cornerRadii=<0 0 0 0>
      }
      tintColor {colorUseRule=UseColor_Specified, specifiedColor=<rgba 1 1 1 1>}
      uVRegion {bIsValid=false, max=<0 0>, min=<0 0>}
     }
     rightShadowBrush {
      bIsDynamicallyLoaded=false
      drawAs=NoDrawType
      imageType=FullColor
      mirroring=NoMirror
      resourceName=../../../Engine/Content/Slate/Common/ScrollBoxShadowRight.png
      resourceObject=null
      tiling=NoTile
      imageSize=<8 16>
      margin=<ltrb 0 0.5 1 0.5>
      outlineSettings {
       bUseBrushTransparency=false
       roundingType=HalfHeightRadius
       width=0
       color {colorUseRule=UseColor_Specified, specifiedColor=<rgba 0 0 0 0>}
       cornerRadii=<0 0 0 0>
      }
      tintColor {colorUseRule=UseColor_Specified, specifiedColor=<rgba 1 1 1 1>}
      uVRegion {bIsValid=false, max=<0 0>, min=<0 0>}
     }
     topShadowBrush {
      bIsDynamicallyLoaded=false
      drawAs=NoDrawType
      imageType=FullColor
      mirroring=NoMirror
      resourceName=../../../Engine/Content/Slate/Common/ScrollBoxShadowTop.png
      resourceObject=null
      tiling=NoTile
      imageSize=<16 8>
      margin=<ltrb 0.5 1 0.5 0>
      outlineSettings {
       bUseBrushTransparency=false
       roundingType=HalfHeightRadius
       width=0
       color {colorUseRule=UseColor_Specified, specifiedColor=<rgba 0 0 0 0>}
       cornerRadii=<0 0 0 0>
      }
      tintColor {colorUseRule=UseColor_Specified, specifiedColor=<rgba 1 1 1 1>}
      uVRegion {bIsValid=false, max=<0 0>, min=<0 0>}
     }
     verticalScrolledContentPadding=<ltrb 0 0 0 1>
    }
   }
   slot {class=VerticalBoxSlot, properties {size {sizeRule=Fill, value=1}}}
  }
  RichText_Body {
   widgetClassPath=&CommonRichTextBlock
   bIsVariable=true
   parent=ScrollBox
   widget=RichText_Body
   properties {
    autoWrapText=true
    bOverrideDefaultStyle=true
    lineHeightPercentage=1.15
    mobileTextBlockScale=1.15
    text=Body Text
    textStyleSet=&DT_Community_RichTextStyles
    defaultTextStyleOverride {
     overflowPolicy=Clip
     transformPolicy=None
     colorAndOpacity {colorUseRule=UseColor_Specified, specifiedColor=<rgba 0.5776 0.5776 0.5776 1>}
     font {
      bForceMonospaced=false
      bMaterialIsStencil=false
      fontMaterial=null
      fontObject=&HeadingNow
      letterSpacing=10
      monospacedWidth=1
      size=15.75
      skewAmount=0
      typefaceFontName=74 Regular
      outlineSettings {
       bApplyOutlineToDropShadows=false
       bMiteredCorners=false
       bSeparateFillAlpha=false
       outlineMaterial=null
       outlineSize=0
       outlineColor=<rgba 0 0 0 1>
      }
     }
     highlightColor {colorUseRule=UseColor_Specified, specifiedColor=<rgba 0 0 0 0>}
     highlightShape {
      bIsDynamicallyLoaded=false
      drawAs=Image
      imageType=NoImage
      mirroring=NoMirror
      resourceName=None
      resourceObject=null
      tiling=NoTile
      imageSize=<32 32>
      margin=<ltrb 0 0 0 0>
      outlineSettings {
       bUseBrushTransparency=false
       roundingType=HalfHeightRadius
       width=0
       color {colorUseRule=UseColor_Specified, specifiedColor=<rgba 0 0 0 0>}
       cornerRadii=<0 0 0 1>
      }
      tintColor {colorUseRule=UseColor_Specified, specifiedColor=<rgba 1 1 1 1>}
      uVRegion {bIsValid=false, max=<0 0>, min=<0 0>}
     }
     selectedBackgroundColor {colorUseRule=UseColor_Foreground, specifiedColor=<rgba 1 0 1 1>}
     shadowColorAndOpacity=<rgba 0 0 0 1>
     shadowOffset=<0 0>
     strikeBrush {
      bIsDynamicallyLoaded=false
      drawAs=Image
      imageType=NoImage
      mirroring=NoMirror
      resourceName=None
      resourceObject=null
      tiling=NoTile
      imageSize=<32 32>
      margin=<ltrb 0 0 0 0>
      outlineSettings {
       bUseBrushTransparency=false
       roundingType=HalfHeightRadius
       width=0
       color {colorUseRule=UseColor_Specified, specifiedColor=<rgba 0 0 0 0>}
       cornerRadii=<0 0 0 1>
      }
      tintColor {colorUseRule=UseColor_Specified, specifiedColor=<rgba 1 1 1 1>}
      uVRegion {bIsValid=false, max=<0 0>, min=<0 0>}
     }
     underlineBrush {
      bIsDynamicallyLoaded=false
      drawAs=Image
      imageType=NoImage
      mirroring=NoMirror
      resourceName=None
      resourceObject=null
      tiling=NoTile
      imageSize=<32 32>
      margin=<ltrb 0 0 0 0>
      outlineSettings {
       bUseBrushTransparency=false
       roundingType=HalfHeightRadius
       width=0
       color {colorUseRule=UseColor_Specified, specifiedColor=<rgba 0 0 0 0>}
       cornerRadii=<0 0 0 1>
      }
      tintColor {colorUseRule=UseColor_Specified, specifiedColor=<rgba 1 1 1 1>}
      uVRegion {bIsValid=false, max=<0 0>, min=<0 0>}
     }
    }
   }
   slot {class=ScrollBoxSlot}
  }
  EntryBox_Buttons {
   widgetClassPath=/Script/FortniteUI.FortDynamicEntryBox
   bIsVariable=true
   parent=VB
   widget=EntryBox_Buttons
   properties {
    entryBoxType=Vertical
    entryWidgetClass=&WBP_UIKit_Button_Loud_C
    navigation=&self:WidgetTree.EntryBox_Buttons.WidgetNavigation_0
    numDesignerPreviewEntries=1
    entrySizeRule {sizeRule=Fill, value=1}
    entrySpacing=<24 24>
   }
   slot {class=VerticalBoxSlot, properties {verticalAlignment=VAlign_Bottom, padding=<ltrb 0 32 0 0>}}
  }
  HB_ReactionReply {
   widgetClassPath=&HorizontalBox
   parent=VB
   widget=HB_ReactionReply
   slot {class=VerticalBoxSlot, properties {padding=<ltrb 0 27 0 0>}}
  }
  Button_Reaction {
   widgetClassPath=&WBP_CommunityPost_ReactionButton
   bIsVariable=true
   parent=HB_ReactionReply
   widget=Button_Reaction
   properties {
    bHideInputAction=true
    icon=null
    minHeight=56
    mobileMinHeight=80
    navigation=&self:WidgetTree.Button_Reaction.WidgetNavigation_0
   }
   slot {class=HorizontalBoxSlot, properties {padding=<ltrb 0 0 24 0>, size {sizeRule=Fill, value=1}}}
  }
  Button_ViewReplies {
   widgetClassPath=&WBP_UIKit_Button_Regular
   bIsVariable=true
   parent=HB_ReactionReply
   widget=Button_ViewReplies
   properties {
    bDisplayInputActionWhenNotInteractable=false
    bHideInputAction=true
    bHideInputActionWithKeyboard=true
    bShouldUseFallbackDefaultInputAction=false
    buttonType=Quiet
    iconMaterial=/FortUILibrary/Textures/MaterialInstances/MI_UIKit_IconDark_ExternalLink
    maxWidth=368
    navigation=&self:WidgetTree.Button_ViewReplies.WidgetNavigation_0
    text="View ## Replies"
   }
   slot {class=HorizontalBoxSlot, properties {horizontalAlignment=HAlign_Left, verticalAlignment=VAlign_Top}}
  }
  Overlay_QRCode {
   widgetClassPath=&Overlay
   parent=GP_HighlightContent
   widget=Overlay_QRCode
   properties {visibility=Collapsed}
   slot {
    class=GridSlot
    properties {columnSpan=2, horizontalAlignment=HAlign_Right, row=1, verticalAlignment=VAlign_Bottom, padding=<ltrb 0 0 60 56>}
   }
  }
  Image_180 {
   widgetClassPath=&Image
   bIsVariable=true
   parent=Overlay_QRCode
   widget=Image_180
   properties {
    visibility=SelfHitTestInvisible
    brush {
     bIsDynamicallyLoaded=false
     drawAs=Image
     imageType=FullColor
     mirroring=NoMirror
     resourceName=None
     resourceObject=&MI_UI_Community_QRBackplate
     tiling=NoTile
     imageSize=<100 100>
     margin=<ltrb 0 0 0 0>
     outlineSettings {
      bUseBrushTransparency=false
      roundingType=HalfHeightRadius
      width=0
      color {colorUseRule=UseColor_Specified, specifiedColor=<rgba 0 0 0 0>}
      cornerRadii=<0 0 0 0>
     }
     tintColor {colorUseRule=UseColor_Specified, specifiedColor=<rgba 1 1 1 1>}
     uVRegion {bIsValid=false, max=<0 0>, min=<0 0>}
    }
   }
   slot {class=OverlaySlot, properties {horizontalAlignment=HAlign_Fill, verticalAlignment=VAlign_Fill}}
  }
  Image_QRCode {
   widgetClassPath=&Image
   bIsVariable=true
   parent=Overlay_QRCode
   widget=Image_QRCode
   properties {
    visibility=SelfHitTestInvisible
    brush {
     bIsDynamicallyLoaded=false
     drawAs=Image
     imageType=FullColor
     mirroring=NoMirror
     resourceName=None
     resourceObject=null
     tiling=NoTile
     imageSize=<90 90>
     margin=<ltrb 0 0 0 0>
     outlineSettings {
      bUseBrushTransparency=false
      roundingType=HalfHeightRadius
      width=0
      color {colorUseRule=UseColor_Specified, specifiedColor=<rgba 0 0 0 0>}
      cornerRadii=<0 0 0 0>
     }
     tintColor {colorUseRule=UseColor_Specified, specifiedColor=<rgba 1 1 1 1>}
     uVRegion {bIsValid=false, max=<0 0>, min=<0 0>}
    }
   }
   slot {class=OverlaySlot, properties {horizontalAlignment=HAlign_Center, verticalAlignment=VAlign_Center}}
  }
  Image_Blur {
   widgetClassPath=&Image
   bIsVariable=true
   parent=Root
   widget=Image_Blur
   properties {
    clipping=ClipToBounds
    pixelSnapping=SnapToPixel
    visibility=SelfHitTestInvisible
    brush {
     bIsDynamicallyLoaded=false
     drawAs=Image
     imageType=FullColor
     mirroring=NoMirror
     resourceName=None
     resourceObject=/FortniteCoreUI/UI/Community/Art/MaterialInstances/MI_UI_Communities_Background
     tiling=NoTile
     imageSize=<32 32>
     margin=<ltrb 0 0 0 0>
     outlineSettings {
      bUseBrushTransparency=false
      roundingType=HalfHeightRadius
      width=0
      color {colorUseRule=UseColor_Specified, specifiedColor=<rgba 0 0 0 0>}
      cornerRadii=<0 0 0 0>
     }
     tintColor {colorUseRule=UseColor_Specified, specifiedColor=<rgba 1 1 1 1>}
     uVRegion {bIsValid=false, max=<0 0>, min=<0 0>}
    }
   }
   slot {class=GridSlot, properties {layer=1, rowSpan=2}}
  }
  GP_Failed {
   widgetClassPath=&GridPanel
   bIsVariable=true
   parent=Root
   widget=GP_Failed
   properties {visibility=Collapsed, columnFill [1], rowFill [1,1]}
   slot {class=GridSlot, properties {layer=2, rowSpan=2}}
  }
  Text_FailedToLoad {
   widgetClassPath=&CommonTextBlock
   bIsVariable=true
   parent=GP_Failed
   widget=Text_FailedToLoad
   properties {
    autoWrapText=true
    bIsScrollingEnabled=false
    justification=Center
    mobileFontSizeMultiplier=1.14
    text=Unable To Load Post
    textTransformPolicy=ToUpper
    font {
     bForceMonospaced=false
     bMaterialIsStencil=false
     fontMaterial=null
     fontObject=&HeadingNow
     letterSpacing=10
     monospacedWidth=1
     size=25.5
     skewAmount=0
     typefaceFontName=86 Bold Italic
     outlineSettings {
      bApplyOutlineToDropShadows=false
      bMiteredCorners=false
      bSeparateFillAlpha=false
      outlineMaterial=null
      outlineSize=0
      outlineColor=<rgba 0 0 0 1>
     }
    }
    shadowOffset=<0 0>
   }
   slot {class=GridSlot, properties {verticalAlignment=VAlign_Bottom}}
  }
  CommonTextBlock {
   widgetClassPath=&CommonTextBlock
   parent=GP_Failed
   widget=CommonTextBlock
   properties {
    autoWrapText=true
    bIsScrollingEnabled=false
    justification=Center
    mobileFontSizeMultiplier=1.14
    text=Refresh this page or try again later
    colorAndOpacity {colorUseRule=UseColor_Specified, specifiedColor=<rgba 0.5776 0.5776 0.5776 1>}
    font {
     bForceMonospaced=false
     bMaterialIsStencil=false
     fontMaterial=null
     fontObject=&HeadingNow
     letterSpacing=0
     monospacedWidth=1
     size=18
     skewAmount=0
     typefaceFontName=86 Bold
     outlineSettings {
      bApplyOutlineToDropShadows=false
      bMiteredCorners=false
      bSeparateFillAlpha=false
      outlineMaterial=null
      outlineSize=2
      outlineColor=<rgba 0 0 0 0.2>
     }
    }
    shadowOffset=<0 0>
   }
   slot {class=GridSlot, properties {row=1, verticalAlignment=VAlign_Top}}
  }
 ]
}
