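# Widget for a single reaction pill on a community post, showing an icon or emoji beside a reaction count.
# PreConstruct applies instance variables: the count as plain integer text, the icon's visibility, material and tint, the emoji's visibility and text, and whether the background shows.
# By default the emoji and background show and the icon is collapsed.
# Construct calls SetCompactUIStyle, which enlarges the horizontal box padding when compact UI is active.
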WidgetBlueprint /FortniteCoreUI/UI/Community/WBP_CommunityPost_ReactionsLine
graphs [
 EventGraph {nodes [
   Event PreConstruct()
   > Set Text(
    Text="To Text (Integer)"(Value=Get ReactionCount() -> (ReactionCount:int=0))
    self=Get Text_Counter() -> (Text_Counter:Common Text Object Reference)
   )
   n1 = Get Image_Icon() -> (Image_Icon:Image Object Reference)
   > Set Visibility(Visibility=Get Brush Visibility() -> (Brush Visibility:ESlateVisibility Enum=Visible), self=@n1)
   > SetBrushFromMaterial(self=@n1, Material=Get Material() -> (Material:Material Interface Object Reference))
   > SetColorAndOpacity(
    self:Image Object Reference=@n1
    InColorAndOpacity:Linear Color Structure=Get In Color and Opacity() -> (In Color and Opacity:Linear Color Structure)
   )
   n2 = Get Text_Emoji() -> (Text_Emoji:Common Text Object Reference)
   > Set Visibility(Visibility=Get Emoji Visibility() -> (Emoji Visibility:ESlateVisibility Enum=Visible), self=@n2)
   > Set Text(Text=Get emoji() -> (emoji:Text), self=@n2)
   > SetVisibility(
    self=Get Image_Background() -> (Image_Background:Image Object Reference)
    InVisibility=Select(Option 0=Collapsed, Option 1=SelfHitTestInvisible, Index=Get Background() -> (Background:bool))
   )
   Event Construct()
   > SetCompactUIStyle() {is=title K2Node_CallFunction}
  ]}
 SetCompactUIStyle {nodes [
   SetCompactUIStyle() {is=title K2Node_FunctionEntry}
   > Branch(Condition=IsUsingCompactUI())
   > SetPadding!(
    execute
    self:Overlay Slot Object Reference=SlotAsOverlaySlot(Widget=Get HB_Reaction() -> (HB_Reaction:Horizontal Box Object Reference))
    InPadding_Left:float=20
    InPadding_Top:float=8
    InPadding_Right:float=20
    InPadding_Bottom:float=8
   ) -> (then)
  ]}
]
variables [
 Background {type=boolean, default=true}
 Brush Visibility {type=ESlateVisibility, default=Collapsed}
 emoji {type=string, default="🔥"}
 Emoji Visibility {type=ESlateVisibility, default=Visible}
 In Color and Opacity {type=LinearColor, default=<rgba 1 0 0.07 1>}
 Material {type=MaterialInterface, default=null}
 ReactionCount {type=integer}
]
widgets {
 info {inheritedWidgetCount=0, parentClass=&UserWidget, rootWidgetClass=&Overlay, widgetCount=6}
 widgets [
  Overlay_0 {widgetClassPath=&Overlay, widget=Overlay_0}
  Image_Background {
   widgetClassPath=&Image
   bIsVariable=true
   parent=Overlay_0
   widget=Image_Background
   properties {brush {
     bIsDynamicallyLoaded=false
     drawAs=Image
     imageType=FullColor
     mirroring=NoMirror
     resourceName=None
     resourceObject=/FortniteCoreUI/UI/Community/Art/MaterialInstances/MI_UIKit_BgQuiet_Reaction
     tiling=NoTile
     imageSize=<32 32>
     margin=<ltrb 0 0 0 0>
     outlineSettings {
      bUseBrushTransparency=false
      roundingType=HalfHeightRadius
      width=0
      color {colorUseRule=UseColor_Specified, specifiedColor=<rgba 0 0 0 0>}
      cornerRadii=<0 0 0 0>
     }
     tintColor {colorUseRule=UseColor_Specified, specifiedColor=<rgba 1 1 1 1>}
     uVRegion {bIsValid=false, max=<0 0>, min=<0 0>}
    }}
   slot {class=OverlaySlot, properties {horizontalAlignment=HAlign_Fill, verticalAlignment=VAlign_Fill}}
  }
  HB_Reaction {
   widgetClassPath=&HorizontalBox
   bIsVariable=true
   parent=Overlay_0
   widget=HB_Reaction
   slot {class=OverlaySlot, properties {verticalAlignment=VAlign_Center, padding=<ltrb 12 4 12 4>}}
  }
  Image_Icon {
   widgetClassPath=&Image
   bIsVariable=true
   parent=HB_Reaction
   widget=Image_Icon
   properties {
    brush {
     bIsDynamicallyLoaded=false
     drawAs=Image
     imageType=NoImage
     mirroring=NoMirror
     resourceName=None
     resourceObject=null
     tiling=NoTile
     imageSize=<24 24>
     margin=<ltrb 0 0 0 0>
     outlineSettings {
      bUseBrushTransparency=false
      roundingType=HalfHeightRadius
      width=0
      color {colorUseRule=UseColor_Specified, specifiedColor=<rgba 0 0 0 0>}
      cornerRadii=<0 0 0 0>
     }
     tintColor {colorUseRule=UseColor_Specified, specifiedColor=<rgba 1 1 1 1>}
     uVRegion {bIsValid=false, max=<0 0>, min=<0 0>}
    }
    colorAndOpacity=<rgba 1 0.5681 0 1>
   }
   slot {
    class=HorizontalBoxSlot
    properties {horizontalAlignment=HAlign_Left, verticalAlignment=VAlign_Center, padding=<ltrb 0 0 6 0>, size {sizeRule=Fill, value=1}}
   }
  }
  Text_Emoji {
   widgetClassPath=&CommonTextBlock
   bIsVariable=true
   parent=HB_Reaction
   widget=Text_Emoji
   properties {
    mobileFontSizeMultiplier=1.15
    text="🔥"
    font {
     bForceMonospaced=false
     bMaterialIsStencil=false
     fontMaterial=null
     fontObject=&NotoSans
     letterSpacing=0
     monospacedWidth=1
     size=13.5
     skewAmount=0
     typefaceFontName=NotoSans-Regular
     outlineSettings {
      bApplyOutlineToDropShadows=false
      bMiteredCorners=false
      bSeparateFillAlpha=false
      outlineMaterial=null
      outlineSize=0
      outlineColor=<rgba 0 0 0 1>
     }
    }
    shadowOffset=<0 0>
   }
   slot {class=HorizontalBoxSlot, properties {horizontalAlignment=HAlign_Left, verticalAlignment=VAlign_Center, padding=<ltrb 0 0 4 0>}}
  }
  Text_Counter {
   widgetClassPath=&CommonTextBlock
   bIsVariable=true
   parent=HB_Reaction
   widget=Text_Counter
   properties {
    mobileFontSizeMultiplier=1.27
    text=285.7K
    textTransformPolicy=ToUpper
    font {
     bForceMonospaced=false
     bMaterialIsStencil=false
     fontMaterial=null
     fontObject=&HeadingNow
     letterSpacing=10
     monospacedWidth=1
     size=14.25
     skewAmount=0
     typefaceFontName=76 Bold
     outlineSettings {
      bApplyOutlineToDropShadows=false
      bMiteredCorners=false
      bSeparateFillAlpha=false
      outlineMaterial=null
      outlineSize=0
      outlineColor=<rgba 0 0 0 1>
     }
    }
    shadowOffset=<0 0>
   }
   slot {class=HorizontalBoxSlot, properties {horizontalAlignment=HAlign_Left, verticalAlignment=VAlign_Center}}
  }
 ]
}
